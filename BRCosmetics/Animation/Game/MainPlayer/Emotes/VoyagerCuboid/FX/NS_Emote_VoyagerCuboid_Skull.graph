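# Niagara system for the skull effect of the VoyagerCuboid emote: a skull of mesh pieces bursts apart amid orange glitch sprites.
# Skull_Break_New places pieces from a skull-vertex mesh using per-index scales, launches them outward, then applies drag and attraction.
# Glitch tracks a skeletal bone early; Dissolve_Glitch sweeps later.
# Disabled Skull_Form, Skull_Break and Lightning emitters hold a 23-piece variant, with lightning linked through particle reads and position arrays.
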NiagaraSystem /BRCosmetics/Animation/Game/MainPlayer/Emotes/VoyagerCuboid/FX/NS_Emote_VoyagerCuboid_Skull
dependencies {
 usedDataInterfaces [
  &NiagaraDataInterfaceMeshRendererInfo
  &NiagaraDataInterfaceSpriteRendererInfo
  &NiagaraDataInterfaceCurve
  &NiagaraDataInterfaceArrayFloat3
  &NiagaraDataInterfaceVector2DCurve
  &NiagaraDataInterfaceParticleRead
  &NiagaraDataInterfaceArrayInt32
  &NiagaraDataInterfaceStaticMesh
  &NiagaraDataInterfaceArrayFloat
  &NiagaraDataInterfaceArrayPosition
  &NiagaraDataInterfaceColorCurve
  &NiagaraDataInterfaceSkeletalMesh
 ]
 usedDynamicInputs []
 usedModules [
  &self:SystemSpawnScript.SystemScriptSource.SystemScriptGraph.NiagaraNodeAssignment_3.SetVariables_#1
  &SystemState
  &self:SystemSpawnScript.SystemScriptSource.SystemScriptGraph.NiagaraNodeAssignment_0.SetVariables_#2
  &self:SystemSpawnScript.SystemScriptSource.SystemScriptGraph.NiagaraNodeAssignment_1.SetVariables_#3
  &self:SystemSpawnScript.SystemScriptSource.SystemScriptGraph.NiagaraNodeAssignment_2.SetVariables_#4
 ]
}
emitters {
 Dissolve_Glitch {
  emitterUpdateScript [
   &EmitterState(Life Cycle Mode=::NewEnumerator1, Loop Behavior=::NewEnumerator1, Loop Delay=1.55, Loop Duration=0.23, UseLoopDelay=true)
   &SpawnBurst_Instantaneous(Spawn Group=1, Use Spawn Probability=false, UseLoopCountLimit=false)
   &SpawnRate(
    SpawnRate=&FloatFromCurve(FloatCurve=DataInterface() {propertyValues {Curve {keys [<at 0.2 1>,<at 1 0>]}}}, CurveIndex=$Emitter.NormalizedLoopAge, Scale Curve=350)
    Use Spawn Probability=false
   )
  ]
  particleSpawnScript [
   &PartitionParticles()
   &InitializeParticle(
    Color=<rgba 5 1.0838 0 1>
    Lifetime=&RandomRangeFloat(Minimum=0.1, Maximum=0.35)
    Sprite Size=&RandomRangeVector2D(Minimum=<10 20>, Maximum=<15 80>)
    Sprite Size Mode=::NewEnumerator4
    Sprite UV Mode=::NewEnumerator1
    UsePositionOffset=false
    Write Material Random=true
   ) {module=InitializeParticle001}
  ]
  particleUpdateScript [
   &ParticleState()
   &ShapeLocation(
    MaskSpawnGroup=false
    Non Uniform Scale=<1 0.1 0.4>
    Offset=&Lerp_Vector(
     A=<0 50 100>
     B=<1 -50 100>
     Alpha=&FloatFromCurve(
      FloatCurve=DataInterface() {propertyValues {Curve {defaultValue=1, keys [{tangentMode=RCTM_User, time=1.6195, value=0}, {tangentMode=RCTM_User, time=1.92, value=1}]}}}
      CurveIndex=$Emitter.Age
     )
    )
    Offset Mode=::NewEnumerator0
    Override Seed=false
    UseOwnerScale=true
   )
   &DynamicMaterialParameters(
    Index 0 Param 1=&RandomRangeFloat(Minimum=0, Maximum=1)
    Index 0 Param 2=&Lerp_Float(A=0, B=0.5, Alpha=$Particles.NormalizedAge)
    Index 0 Param 3=/Niagara/DynamicInputs/Integer/V2/Round_Float(Float:NiagaraFloat=&RandomRangeFloat(Minimum=0, Maximum=3))
    Module.Index0Float0Write=true
    Module.Index0Float1Write=true
    Module.Index0Float2Write=true
    Module.Index0Float3Write=true
   )
   &ScaleSpriteSize(
    Scale Factor=&Vector2DFromCurve(
     Vector2Curve=DataInterface() {propertyValues {
       XCurve {keys [<at 0 1>,<at 1 1>]}
       YCurve {keys [{arriveTangent=4.2107, interpMode=RCIM_Cubic, leaveTangent=4.2107, tangentMode=RCTM_User, time=0, value=0.5}, <at 1 2 RCIM_Cubic>]}
      }}
     ScaleCurve=&Multiply_Vector2D_ByFloat(Float=0.8, Vector2D=&RandomRangeVector2D(Minimum=<0.8 0.8>, Maximum=<1.3 1.3>))
    )
    Scale Sprite Size Mode=::NewEnumerator2
   )
   &ScaleColor(
    Scale Alpha=&FloatFromCurve(FloatCurve=DataInterface() {propertyValues {Curve {keys [
         {arriveTangent=3.8564, interpMode=RCIM_Cubic, leaveTangent=3.8564, tangentMode=RCTM_User, time=0, value=0.5274}
         {arriveTangent=0.1295, interpMode=RCIM_Cubic, leaveTangent=0.1295, tangentMode=RCTM_User, time=0.5, value=1}
         <at 1 0 RCIM_Cubic>
        ]}}})
    ScaleA=true
    ScaleRGB=false
   )
   &SolveForcesAndVelocity(Clamp Velocity=false, Limit Acceleration=false)
   /SetVariables(Particles.GlowColor=&MultiplyLinearColor(
     Linear Color=$Particles.Initial.GlowColor
     Scale Factor=&MakeVector4(
      X=1
      Y=1
      Z=1
      W=&FloatFromCurve(FloatCurve=DataInterface() {propertyValues {Curve {
          defaultValue=1
          keys [
           {interpMode=RCIM_Cubic, tangentMode=RCTM_User, time=0, value=0}
           {interpMode=RCIM_Cubic, tangentMode=RCTM_User, time=0.1662, value=1}
           {tangentMode=RCTM_User, time=1, value=0}
          ]
         }}})
     )
    )) {isSetParameters=true, module=SetVariables_#5}
  ]
  rendererClasses [&NiagaraSpriteRendererProperties]
  renderers [{rendererClass=&NiagaraSpriteRendererProperties}]
 }
 Glitch {
  emitterUpdateScript [
   &EmitterState(Life Cycle Mode=::NewEnumerator1, Loop Behavior=::NewEnumerator1, Loop Duration=1, UseLoopDelay=true)
   &SpawnBurst_Instantaneous(Spawn Group=1, Use Spawn Probability=false, UseLoopCountLimit=false)
   &SpawnRate(
    SpawnRate=&FloatFromCurve(FloatCurve=DataInterface() {propertyValues {Curve {keys [<at 0.2 1>,<at 0.5 0>]}}}, CurveIndex=$Emitter.NormalizedLoopAge, Scale Curve=70)
    Use Spawn Probability=false
   )
  ]
  particleSpawnScript [
   /SetVariables(Particles.VisibilityTag=$Emitter.SpawnGroup) {isSetParameters=true}
   &InitializeParticle(
    Color=<rgba 5 1.0838 0 1>
    Lifetime=&MergeHelper_FloatSwitch(
     0=&RandomRangeFloat(Minimum=0.25, Maximum=0.6)
     1=1
     2=?
     3=?
     Index=$Particles.VisibilityTag
     Num=::NewEnumerator0
    )
    Sprite Size=&RandomRangeVector2D(Minimum=<10 20>, Maximum=<15 80>)
    Sprite Size Mode=::NewEnumerator4
    Sprite UV Mode=::NewEnumerator1
    UsePositionOffset=false
    Write Material Random=true
   ) {module=InitializeParticle001}
   &ShapeLocation(MaskSpawnGroup=false, Override Seed=false, UseOwnerScale=true)
   /SetVariables(
    Particles.GlowAlignment:Vector3f=<1 0 0>
    Particles.GlowColor=<rgba 1 0.3289 0.0575 0.25>
    Particles.GlowFacing:Vector3f=<0 0 1>
    Particles.GlowPosition:NiagaraPosition=&AddVectorToPosition(Position=$Engine.Emitter.SimulationPosition, Vector=<0 90 1>)
    Particles.GlowSize=<50 150>
   ) {isSetParameters=true}
  ]
  particleUpdateScript [
   &ParticleState()
   &SkeletalMeshLocation(
    Bone / Socket Index=3
    Bone Sampling Mode=::NewEnumerator4
    MaskSpawnGroup=false
    Sampled Position Offset=&RandomRangeVector(Minimum=<-40 90 0>, Maximum=<40 90 0>)
    Skeletal Mesh=$System.Skeletal Mesh Reader
   )
   &DynamicMaterialParameters(
    Index 0 Param 1=&RandomRangeFloat(Minimum=0, Maximum=1)
    Index 0 Param 2=&Lerp_Float(A=0, B=0.5, Alpha=$Particles.NormalizedAge)
    Index 0 Param 3=&Round_Float(Float:NiagaraFloat=&RandomRangeFloat(Minimum=0, Maximum=3))
    Module.Index0Float0Write=true
    Module.Index0Float1Write=true
    Module.Index0Float2Write=true
    Module.Index0Float3Write=true
   )
   &ScaleSpriteSize(
    Scale Factor=&Vector2DFromCurve(
     Vector2Curve=DataInterface() {propertyValues {
       XCurve {keys [<at 0 1>,<at 1 1>]}
       YCurve {keys [{arriveTangent=4.2107, interpMode=RCIM_Cubic, leaveTangent=4.2107, tangentMode=RCTM_User, time=0, value=0.5}, <at 1 2 RCIM_Cubic>]}
      }}
     ScaleCurve=&RandomRangeVector2D(Minimum=<0.8 0.8>, Maximum=<1.3 1.3>)
    )
    Scale Sprite Size Mode=::NewEnumerator2
   )
   &ScaleColor(
    Scale Alpha=&FloatFromCurve(FloatCurve=DataInterface() {propertyValues {Curve {keys [
         {arriveTangent=3.8564, interpMode=RCIM_Cubic, leaveTangent=3.8564, tangentMode=RCTM_User, time=0, value=0.5274}
         {arriveTangent=0.1295, interpMode=RCIM_Cubic, leaveTangent=0.1295, tangentMode=RCTM_User, time=0.5, value=1}
         <at 1 0 RCIM_Cubic>
        ]}}})
    ScaleA=true
    ScaleRGB=false
   )
   &SolveForcesAndVelocity(Clamp Velocity=false, Limit Acceleration=false)
   /SetVariables(Particles.GlowColor=&MultiplyLinearColor(
     Linear Color=$Particles.Initial.GlowColor
     Scale Factor=&MakeVector4(
      X=1
      Y=1
      Z=1
      W=&FloatFromCurve(FloatCurve=DataInterface() {propertyValues {Curve {
          defaultValue=1
          keys [
           {interpMode=RCIM_Cubic, tangentMode=RCTM_User, time=0, value=0}
           {interpMode=RCIM_Cubic, tangentMode=RCTM_User, time=0.1662, value=1}
           {tangentMode=RCTM_User, time=1, value=0}
          ]
         }}})
     )
    )) {isSetParameters=true, module=SetVariables_#5}
  ]
  rendererClasses [&NiagaraSpriteRendererProperties]
  renderers [
   {rendererClass=&NiagaraSpriteRendererProperties}
   {rendererClass=&NiagaraSpriteRendererProperties, rendererIndex=1}
  ]
 }
 Lightning_Break {
  bEnabled=false
  emitterUpdateScript [
   &EmitterState(
    Inactive Response=::NewEnumerator1
    Life Cycle Mode=::NewEnumerator1
    Loop Behavior=::NewEnumerator1
    Loop Delay=1.2
    Loop Duration=2
    UseLoopDelay=true
   )
   &SpawnParticlesFromOtherEmitter(
    Attribute Reader=DataInterface() {propertyValues {EmitterBinding {bindingMode=Other, emitterName=Skull_Break}}}
    Maximum Age Spawn Threshold=1.1
    Minimum Age Spawn Threshold=0
    Spawn Rate=20
    Spawning Enabled:NiagaraBool=&Boolean_AndOperation(
     Bool1=&SetBoolByFloatComparison(A=&MakeFloatFromInt(Int=$Engine.Emitter.NumParticles), B=1, Comparison Type=::NewEnumerator6)
     Bool2=&SetBoolByFloatComparison(A=$Emitter.Age, B=1.4, Comparison Type=::NewEnumerator6)
    )
    Use Maximum Age=true
    Use Minimum Age=false
   )
  ]
  particleSpawnScript [
   &PartitionParticles()
   &InitializeParticle(
    Color Maximum=<rgba 1.25 0.2203 0 1>
    Color Minimum=<rgba 1 0.1669 0.0006 1>
    Color Mode=::NewEnumerator2
    Lifetime=&RandomRangeFloat(Minimum=0.27, Maximum=0.4)
    Mesh Scale Mode=::NewEnumerator3
    UsePositionOffset=false
    Write Material Random=true
   )
   &SampleParticlesFromOtherEmitter(Dynamic Material Parameter 3 Sampling:ENiagara_DirectReadSamplingMode=::NewEnumerator0, Particle ID Sampling=::NewEnumerator1)
   /SetVariables(Particles.FriendSelect:NiagaraInt32=&MergeHelper_IntSwitch(
     0=&RandomRangeInt(Minimum=0, Maximum=10)
     1=&RandomRangeInt(Minimum=12, Maximum=23)
     2=?
     3=?
     4=?
     5=?
     Index=&MakeCustomIntFromBool(
      Bool=&SetBoolByIntComparison(
       A=&MakeIntFromFloat(Float=&MakeFloatFromVector4(Vector 4=$Output.SampleParticlesFromOtherEmitter.DynamicMaterialParameter3, Channel=::NewEnumerator0))
       B=0
       Comparison Type=::NewEnumerator0
      )
      True INT=1
      False INT=0
     )
     Num=::NewEnumerator0
    )) {isSetParameters=true, module=SetVariables_#6}
  ]
  particleUpdateScript [
   &ParticleState()
   &UpdateMeshOrientation(
    Facing Interpolation=false
    Orientation Method=::NewEnumerator2
    Override Quat=false
    TransformFacingOrigin=false
    TransformFacingPosition=false
    TransformUp=false
   )
   /Niagara/Modules/AttributeReader/UpdateParticlesFromOtherEmitter(Attribute Reader=DataInterface() {propertyValues {EmitterBinding {bindingMode=Other, emitterName=Skull_Break}}})
   &SolveForcesAndVelocity(Clamp Velocity=false, Limit Acceleration=false)
   &self:ReadDeath(ID:NiagaraInt32=$Particles.FriendSelect, PositionArray:NiagaraDataInterfaceArrayPosition=$System.Position Array) {module=ReadDeath}
   &DynamicMaterialParameters(
    Index 0 Param 1=&FloatFromCurve(
     FloatCurve=DataInterface() {propertyValues {Curve {keys [<at 0 1>,<at 1 0>]}}}
     Scale Curve=&Multiply_Float(
      A=&UniformRangedFloat(Minimum=15, Maximum=20)
      B=&ClampFloat(Float=&Divide_Float(A=$Particles.LightningLength, B=70), Max=2)
     )
    )
    Index 0 Param 2=&UniformRangedFloat(Minimum=0.2, Maximum=0.2)
    Index 0 Param 3=&UniformRangedFloat(Minimum=0, Maximum=0)
    Index 0 Param 4=&UniformRangedFloat(Minimum=0, Maximum=50)
    Index 1 Param 1=&UniformRangedFloat(Minimum=0.5, Maximum=1)
    Index 1 Param 2=&UniformRangedFloat(Minimum=0, Maximum=20)
    Module.Index0Float0Write=true
    Module.Index0Float1Write=true
    Module.Index0Float2Write=true
    Module.Index0Float3Write=true
    Module.Index1Float0Write=true
    Module.Index1Float1Write=true
    Module.Index1Float2Write=true
    Module.Index1Float3Write=true
    Write Parameter Index 1=::NewEnumerator1
   )
   &ScaleMeshSize(Scale Factor=&MakeVector(X=&Divide_Float(A=$Particles.LightningLength, B=100), Y=1, Z=1))
   &ScaleColor(
    Scale Alpha=&FloatFromCurve(FloatCurve=DataInterface() {propertyValues {Curve {keys [<at 0 0>,<at 0.1671 1>,<at 0.2183 1>,<at 1 0>]}}})
    Scale RGB=<5 5 5>
    ScaleA=true
    ScaleRGB=true
   )
  ]
  rendererClasses [&NiagaraMeshRendererProperties]
  renderers [{rendererClass=&NiagaraMeshRendererProperties, rendererIndex=0}]
 }
 Lightning_Form {
  bEnabled=false
  emitterUpdateScript [
   &EmitterState(
    Inactive Response=::NewEnumerator1
    Life Cycle Mode=::NewEnumerator1
    Loop Behavior=::NewEnumerator1
    Loop Delay=0.5
    Loop Duration=2
    UseLoopDelay=true
   )
   &SpawnParticlesFromOtherEmitter(
    Attribute Reader=DataInterface() {propertyValues {EmitterBinding {bindingMode=Other, emitterName=Skull_Form}}}
    Maximum Age Spawn Threshold=1.1
    Minimum Age Spawn Threshold=0
    Spawn Rate=20
    Spawning Enabled:NiagaraBool=&SetBoolByIntComparison(A=$Engine.Emitter.NumParticles, B=1, Comparison Type=::NewEnumerator6)
    Use Maximum Age=true
    Use Minimum Age=false
   )
  ]
  particleSpawnScript [
   &PartitionParticles()
   &InitializeParticle(
    Color Maximum=<rgba 1.25 0.2203 0 1>
    Color Minimum=<rgba 1 0.1669 0.0006 1>
    Color Mode=::NewEnumerator2
    Lifetime=&RandomRangeFloat(Minimum=0.27, Maximum=0.4)
    Mesh Scale Mode=::NewEnumerator3
    UsePositionOffset=false
    Write Material Random=true
   )
   &SampleParticlesFromOtherEmitter(Dynamic Material Parameter 3 Sampling:ENiagara_DirectReadSamplingMode=::NewEnumerator0, Particle ID Sampling=::NewEnumerator1)
   /SetVariables(Particles.FriendSelect:NiagaraInt32=&MergeHelper_IntSwitch(
     0=&RandomRangeInt(Minimum=0, Maximum=10)
     1=&RandomRangeInt(Minimum=12, Maximum=23)
     2=?
     3=?
     4=?
     5=?
     Index=&MakeCustomIntFromBool(
      Bool=&SetBoolByIntComparison(
       A=&MakeIntFromFloat(Float=&MakeFloatFromVector4(Vector 4=$Output.SampleParticlesFromOtherEmitter.DynamicMaterialParameter3, Channel=::NewEnumerator0))
       B=0
       Comparison Type=::NewEnumerator0
      )
      True INT=1
      False INT=0
     )
     Num=::NewEnumerator0
    )) {isSetParameters=true, module=SetVariables_#6}
  ]
  particleUpdateScript [
   &ParticleState()
   &UpdateMeshOrientation(
    Facing Interpolation=false
    Orientation Method=::NewEnumerator2
    Override Quat=false
    TransformFacingOrigin=false
    TransformFacingPosition=false
    TransformUp=false
   )
   &UpdateParticlesFromOtherEmitter(Attribute Reader=DataInterface() {propertyValues {EmitterBinding {bindingMode=Other, emitterName=Skull_Form}}})
   &SolveForcesAndVelocity(Clamp Velocity=false, Limit Acceleration=false)
   &self:ReadDeath(
    ID:NiagaraInt32=&SelectIntFromArray(
     Direct Array Index=$Engine.ExecIndex
     Int Selection Array=DataInterface() {propertyValues {IntData [11,0,1,2,1,4,2,6,5,8,9,12,13,14,15,17,3,22,11,10,19,20,21]}}
    )
    PositionArray:NiagaraDataInterfaceArrayPosition=$System.Position Array Init
   ) {module=ReadDeath001}
   &DynamicMaterialParameters(
    Index 0 Param 1=&FloatFromCurve(
     FloatCurve=DataInterface() {propertyValues {Curve {keys [<at 0 1>,<at 1 0>]}}}
     Scale Curve=&Multiply_Float(
      A=&UniformRangedFloat(Minimum=15, Maximum=20)
      B=&ClampFloat(Float=&Divide_Float(A=$Particles.LightningLength, B=70), Max=2)
     )
    )
    Index 0 Param 2=&UniformRangedFloat(Minimum=0.2, Maximum=0.2)
    Index 0 Param 3=&UniformRangedFloat(Minimum=0, Maximum=0)
    Index 0 Param 4=&UniformRangedFloat(Minimum=0, Maximum=50)
    Index 1 Param 1=&UniformRangedFloat(Minimum=0.5, Maximum=1)
    Index 1 Param 2=&UniformRangedFloat(Minimum=0, Maximum=20)
    Module.Index0Float0Write=true
    Module.Index0Float1Write=true
    Module.Index0Float2Write=true
    Module.Index0Float3Write=true
    Module.Index1Float0Write=true
    Module.Index1Float1Write=true
    Module.Index1Float2Write=true
    Module.Index1Float3Write=true
    Write Parameter Index 1=::NewEnumerator1
   )
   &ScaleMeshSize(Scale Factor=&MakeVector(X=&Divide_Float(A=$Particles.LightningLength, B=100), Y=1, Z=1))
   &ScaleColor(
    Scale Alpha=&FloatFromCurve(FloatCurve=DataInterface() {propertyValues {Curve {keys [<at 0.2183 1>,<at 1 0>]}}})
    Scale RGB=<5 5 5>
    ScaleA=true
    ScaleRGB=true
   )
  ]
  rendererClasses [&NiagaraMeshRendererProperties]
  renderers [{rendererClass=&NiagaraMeshRendererProperties, rendererIndex=0}]
 }
 Skull_Break {
  bEnabled=false
  emitterUpdateScript [
   &EmitterState()
   &SpawnBurst_Instantaneous(Spawn Count=23, Spawn Time=1.3, Use Spawn Probability=false, UseLoopCountLimit=false)
  ]
  particleSpawnScript [
   &PartitionParticles()
   &InitializeParticle(
    Color Maximum=<rgba 1.25 1.1493 0.9825 1>
    Color Minimum=<rgba 0.6 0.6 0.6 1>
    Color Mode=::NewEnumerator2
    Lifetime=&RandomRangeFloat(Minimum=0.75, Maximum=1.1)
    Mesh Scale Mode=::NewEnumerator3
    Mesh Uniform Scale=4
    Sprite Size Mode=::NewEnumerator3
    Uniform Sprite Size=35
    UsePositionOffset=false
    Write Material Random=true
   )
   &SubUVAnimation(
    Mesh Renderer=DataInterface() {propertyValues {MeshRenderer=&self:Skull_Break_0.NiagaraMeshRendererProperties_0}}
    Output Alternate Renderer Binding=true
    Random Start Frame=true
    Renderer Type:ENiagaraMeshOrSprite=::NewEnumerator1
    SubUV Lookup Index Scale=0
    UseEndFrame=false
    UseStartFrame=false
   )
   &SkeletalMeshLocation(
    Bone / Socket Index=3
    Bone Sampling Mode=::NewEnumerator4
    MaskSpawnGroup=false
    Sampled Position Offset=&Add_Vector(
     A=&Add_Vector(
      A=$User.SkullOffset
      B=&Multiply_Vector(
       A=&VectorFromFloat(Value=0.6)
       B=&SelectVectorFromArray(
        Direct Array Index=&ReturnExecIndex()
        Vector Selection Array=DataInterface() {propertyValues {FloatData [
           <34 0.01 -57>
           <47 0.1 -35>
           <24 -0.02 -34.5>
           <0 0.12 -33>
           <64 -0.06 -27>
           <63 0.02 -5>
           <11.5 -0.12 -16>
           <12.5 0.13 8>
           <59.5 -0.03 18>
           <35 0.15 24.5>
           <15.5 0.08 29.5>
           <-1.5 -0.08 -58>
           <-33 -0.09 -57>
           <-25 -0.01 -34>
           <-48 -0.07 -35>
           <-63.5 0.05 -31>
           <-3 0.04 -16>
           <-63 0 -11.5>
           <-6.5 -0.14 9>
           <-9 0.04 29.5>
           <-31 -0.1 24>
           <-52 0.11 12.5>
           <-62 -0.05 6>
          ]}}
       )
      )
     )
     B=&Lerp_Vector(
      A=&RandomRangeVector(Minimum=<0 0 0>, Maximum=<0 0 0>)
      B=<0 0 0>
      Alpha=&FloatFromCurve(FloatCurve=DataInterface() {propertyValues {Curve {
          defaultValue=1
          keys [{arriveTangent=3.5381, interpMode=RCIM_Cubic, leaveTangent=3.5381, tangentMode=RCTM_User, time=0, value=0}, <at 0.3 1 RCIM_Cubic>]
         }}})
     )
    )
    Skeletal Mesh=$System.Skeletal Mesh Reader
   )
   /SetVariables(Particles.SideToLaunch:NiagaraInt32=&MergeHelper_IntSwitch(
     0=&self:Merge_Helper_Int_Switch_16(
      0:NiagaraInt32=0
      1:NiagaraInt32=0
      2:NiagaraInt32=0
      3:NiagaraInt32=0
      4:NiagaraInt32=0
      5:NiagaraInt32=0
      6:NiagaraInt32=0
      7:NiagaraInt32=0
      8:NiagaraInt32=0
      9:NiagaraInt32=0
      10:NiagaraInt32=0
      11:NiagaraInt32=1
      12:NiagaraInt32=1
      13:NiagaraInt32=1
      14:NiagaraInt32=1
      Index:NiagaraInt32=&ReturnExecIndex()
      Num:NiagaraInt32=13
     )
     1=&MergeHelper_IntSwitch(
      0=1
      1=1
      2=1
      3=1
      4=1
      5=1
      6=1
      7=1
      Index=/Niagara/DynamicInputs/Subtract/Subtract_Integer(A:NiagaraInt32=&ReturnExecIndex(), B:NiagaraInt32=15)
      Num=::NewEnumerator6
     )
     2=?
     3=?
     4=?
     5=?
     Index=&MakeCustomIntFromBool(
      Bool=&SetBoolByIntComparison(A=&ReturnExecIndex(), B=15, Comparison Type=::NewEnumerator2)
      True INT=1
      False INT=0
     )
     Num=::NewEnumerator0
    )) {isSetParameters=true}
   &AddVelocity(Velocity=&MakeVector(
     X=&Multiply_Float(
      A=&RandomRangeFloat(Minimum=-80, Maximum=-150)
      B=&MergeHelper_FloatSwitch(0=-3, 1=3, 2=?, 3=?, Index=$Particles.SideToLaunch, Num=::NewEnumerator0)
     )
     Y=&RandomRangeFloat(Minimum=-80, Maximum=-500)
     Z=&RandomRangeFloat(Minimum=10, Maximum=150)
    ))
   &MeshRotationForce(Rotation=&RandomRangeVector(Minimum=<15 15 15>, Maximum=<-15 -15 -15>))
   &ApplyInitialForces()
   /SetVariables(Particles.GlowColor=<rgba 1 0.526 0.1188 1>) {isSetParameters=true}
  ]
  particleUpdateScript [
   &ParticleState()
   &ScaleMeshSize(Scale Factor=&VectorFromFloat(Value=&FloatFromCurve(
      FloatCurve=DataInterface() {propertyValues {Curve {
         defaultValue=0
         keys [{interpMode=RCIM_Cubic, tangentMode=RCTM_User, time=0.686, value=1}, {interpMode=RCIM_Cubic, tangentMode=RCTM_User, time=1, value=0}]
        }}}
      Scale Curve=0.6
     )))
   &DynamicMaterialParameters(
    Index 0 Param 1=&FloatFromCurve(FloatCurve=DataInterface() {propertyValues {Curve {keys [<at 0 0>,<at 1 4>]}}})
    Index 0 Param 2=1
    Index 0 Param 3=&Waveform("Waveform [1]"=::NewEnumerator0, "[1] Frequency"=0.4, "[1] Amplitude Min/Max"=<1 2>)
    Index 3 Param 1=&MakeFloatFromInt(Int=$Particles.Initial.SideToLaunch)
    Module.Index0Float0Write=true
    Module.Index0Float1Write=true
    Module.Index0Float2Write=true
    Module.Index0Float3Write=true
    Module.Index3Float0Write=true
    Module.Index3Float1Write=false
    Module.Index3Float2Write=false
    Module.Index3Float3Write=false
    Write Parameter Index 3=::NewEnumerator1
   )
   &GravityForce(Gravity=&MakeVector(X=0, Y=0, Z=&RandomRangeFloat(Minimum=-550, Maximum=-980)))
   &Drag(
    Drag=&FloatFromCurve(
     FloatCurve=DataInterface() {propertyValues {Curve {keys [<at 0 0>,<at 0.3018 3>]}}}
     Scale Curve=&RandomRangeFloat(Minimum=0.5, Maximum=1.5)
    )
    Use Linear Drag=true
    Use Rotational Drag=true
   )
   &SolveForcesAndVelocity(Clamp Velocity=false, Limit Acceleration=false)
   &self:WriteDeath(ID:NiagaraInt32=&ReturnExecIndex(), Position Array:NiagaraDataInterfaceArrayPosition=$System.Position Array) {module=WriteDeath}
   /SetVariables(Particles.GlowColor=&MultiplyLinearColor(
     Linear Color=$Particles.Initial.GlowColor
     Scale Factor=&MakeVector4(
      X=1
      Y=1
      Z=1
      W=&FloatFromCurve(
       FloatCurve=DataInterface() {propertyValues {Curve {
          defaultValue=1
          keys [{interpMode=RCIM_Cubic, tangentMode=RCTM_User, time=0.2127, value=1}, {interpMode=RCIM_Cubic, tangentMode=RCTM_User, time=0.4934, value=0}]
         }}}
       Scale Curve=0.2
      )
     )
    )) {isSetParameters=true}
   &ScaleColor(
    Scale RGB=&VectorFromFloat(Value=&FloatFromCurve(
      FloatCurve=DataInterface() {propertyValues {Curve {
         defaultValue=1
         keys [{interpMode=RCIM_Cubic, tangentMode=RCTM_User, time=0, value=1}, {interpMode=RCIM_Cubic, tangentMode=RCTM_User, time=1, value=0}]
        }}}
      Scale Curve=&RandomRangeFloat(Minimum=1.5, Maximum=2.5)
     ))
    ScaleA=true
    ScaleRGB=true
   )
  ]
  rendererClasses [&NiagaraMeshRendererProperties, &NiagaraSpriteRendererProperties]
  renderers [
   {rendererClass=&NiagaraMeshRendererProperties, rendererIndex=0}
   {rendererClass=&NiagaraSpriteRendererProperties, rendererIndex=1}
  ]
 }
 Skull_Break_New {
  emitterUpdateScript [
   &EmitterState(
    Inactive Response=::NewEnumerator1
    Life Cycle Mode=::NewEnumerator1
    Loop Behavior=::NewEnumerator1
    Loop Duration Mode=::NewEnumerator1
    UseLoopDelay=false
   )
   &SpawnBurst_Instantaneous(Spawn Count=110, Spawn Time=1.3, Use Spawn Probability=false, UseLoopCountLimit=false) {module=SpawnBurst_Instantaneous001}
  ]
  particleSpawnScript [
   &PartitionParticles()
   &self:ScratchModule_03(
    Skull Offset:Vector3f=&Subtract_Vector(A=$User.SkullOffset, B=<0 0 9>)
    Static Mesh Reader=DataInterface() {propertyValues {
      DefaultMesh=&/SM_Emote_VoyagerCuboid_SkullVerts4
      InstanceIndex=-1
      LODIndex=0
      PreviewMesh=&SM_Emote_VoyagerCuboid_SkullVerts4
      SoftSourceActor=null
      SourceMode=DefaultMeshOnly
      bAllowSamplingFromStreamingLODs=false
      bCaptureTransformsPerFrame=true
      bUsePhysicsBodyVelocity=false
      FilteredSockets []
      LODIndexUserParameter {parameter {name=None, typeDefHandle {registeredTypeIndex=108}, varData []}}
      MeshParameterBinding {parameter {name=None, typeDefHandle {registeredTypeIndex=1}, varData []}}
      SectionFilter {allowedMaterialSlots []}
     }}
    TriangleID:NiagaraInt32=&ReturnExecIndex()
   ) {module=ScratchModule_03}
   /SetVariables(Particles.Scale Array=DataInterface() {propertyValues {
      GpuSyncMode=SyncCpuToGpu
      MaxElements=0
      FloatData [2.6,2.1,2,2.8,2.6,1.8,1.9,1.2,2,1.1,1.1,3.2,1.3,1.7,1.5,2.5,2.5,2.4,1,0.9,1.7,0.9,0.9,3.5,1.5,1.7,1.5,1,1,1,1.5,1,1.5,1,2,3.5,3.5,1.2,1.4,2.5,2,1.3,1.2,1,2,2,2,3,2.9,2,1.9,2.7,1.2,2.6,2.55,3.2,1.5,1.4,3.5,3,3.2,3.5,3.7,2.7,1.6,1.5,2.3,2.2,1.2,1.2,1.6,2,2.1,2,2,1.6,3,2.5,3.7,3.5,1.3,2.6,3.3,3,3.5,4,3.2,4.5,4,2,2.7,2.5,2,2,1.5,2,2.6,1.8,1.7,1,2.5,2.8,1,2,1,1,1,1.3,1.25,2]
     }}) {isSetParameters=true, module=SetVariables_#7}
   &InitializeParticle(
    Color Maximum=<rgba 1.25 1.1493 0.9825 1>
    Color Minimum=<rgba 0.6 0.6 0.6 1>
    Color Mode=::NewEnumerator2
    Lifetime=&RandomRangeFloat(Minimum=0.4, Maximum=0.7)
    Mesh Scale Mode=::NewEnumerator3
    Mesh Uniform Scale=&Multiply_Float(
     A=&SelectFloatFromArray(Direct Array Index=&ReturnExecIndex(), Float Selection Array=$Particles.Scale Array)
     B=$User.Skull Scale
    )
    Position Offset=&Multiply_VectorByFloat(Vector=Linked() {linkedVariable {name=Particles.MeshIDPos, type=NiagaraPosition}}, Float=$User.Skull Scale)
    Sprite Size Mode=::NewEnumerator3
    Uniform Sprite Size=45
    UsePositionOffset=true
    Write Material Random=true
   )
   &SubUVAnimation(
    Mesh Renderer=DataInterface() {propertyValues {MeshRenderer=&self:Skull_Break_New_0.NiagaraMeshRendererProperties_0}}
    Output Alternate Renderer Binding=true
    Random Start Frame=true
    Renderer Type:ENiagaraMeshOrSprite=::NewEnumerator1
    SubUV Lookup Index Scale=0
    UseEndFrame=false
    UseStartFrame=false
   )
   /SetVariables(Particles.GlowColor=<rgba 1 0.526 0.1188 1>) {isSetParameters=true, module=SetVariables_#8}
   &AddVelocity(Module.Use Falloff Distance=false, Origin Offset=<0 87 140>, Override Seed=false, Velocity Mode=::NewEnumerator1, Velocity Speed=200) {module=AddVelocity001}
   &AddVelocity(Velocity=&MakeCustomVectorFromBool(
     Bool=&SetBoolByFloatComparison(
      A=&MakeFloatFromVector(Vector=Linked() {linkedVariable {name=Particles.MeshIDPos, type=NiagaraPosition}}, Channel=::NewEnumerator0)
      B=0
     )
     True Vector=<110 0 1>
     False Vector=<-110 0 0>
    ))
   &ApplyInitialForces()
  ]
  particleUpdateScript [
   &ParticleState()
   &ScaleMeshSize(Scale Factor=&MakeVector(
     X=&FloatFromCurve(FloatCurve=DataInterface() {propertyValues {Curve {
         defaultValue=1
         keys [
          {interpMode=RCIM_Cubic, tangentMode=RCTM_User, time=0.1799, value=1}
          {arriveTangent=-3.0696, interpMode=RCIM_Cubic, leaveTangent=-3.0696, tangentMode=RCTM_User, time=1, value=0}
         ]
        }}})
     Y=&FloatFromCurve(FloatCurve=DataInterface() {propertyValues {Curve {
         defaultValue=1
         keys [
          {interpMode=RCIM_Cubic, tangentMode=RCTM_User, time=0.1733, value=1}
          {arriveTangent=-2.8856, interpMode=RCIM_Cubic, leaveTangent=-2.8856, tangentMode=RCTM_User, time=1, value=0}
         ]
        }}})
     Z=&FloatFromCurve(
      FloatCurve=DataInterface() {propertyValues {Curve {
         defaultValue=1
         keys [
          {interpMode=RCIM_Cubic, tangentMode=RCTM_User, time=0.6, value=1}
          {arriveTangent=7.3578, interpMode=RCIM_Cubic, leaveTangent=7.3578, tangentMode=RCTM_User, time=0.9527, value=2}
         ]
        }}}
      Scale Curve=&Lerp_Float(
       A=1
       B=&RandomRangeFloat(Minimum=1, Maximum=2)
       Alpha=&FloatFromCurve(FloatCurve=DataInterface() {propertyValues {Curve {
           defaultValue=1
           keys [
            {interpMode=RCIM_Cubic, tangentMode=RCTM_User, time=0.4398, value=0}
            {arriveTangent=4.8675, interpMode=RCIM_Cubic, leaveTangent=4.8675, tangentMode=RCTM_User, time=0.9527, value=1}
           ]
          }}})
      )
     )
    )) {module=ScaleMeshSize001}
   &ScaleMeshSize(Scale Factor=&VectorFromFloat(Value=&FloatFromCurve(
      FloatCurve=DataInterface() {propertyValues {Curve {
         defaultValue=1
         keys [
          {arriveTangent=4.1785, interpMode=RCIM_Cubic, leaveTangent=4.1785, tangentMode=RCTM_User, time=0, value=0}
          {arriveTangent=3.7835, interpMode=RCIM_Cubic, leaveTangent=3.7835, tangentMode=RCTM_User, time=0.2136, value=0.6422}
          {arriveTangent=0.0224, interpMode=RCIM_Cubic, leaveTangent=0.0224, tangentMode=RCTM_User, time=0.411, value=1}
         ]
        }}}
      Scale Curve=0.6
     ))) {enabled=false}
   &DynamicMaterialParameters(
    Index 0 Param 1=&FloatFromCurve(FloatCurve=DataInterface() {propertyValues {Curve {keys [<at 0 0>,<at 1 4>]}}})
    Index 0 Param 2=1
    Index 0 Param 3=&Waveform("Waveform [1]"=::NewEnumerator0, "[1] Frequency"=0.4, "[1] Amplitude Min/Max"=<1 2>)
    Module.Index0Float0Write=true
    Module.Index0Float1Write=true
    Module.Index0Float2Write=true
    Module.Index0Float3Write=true
   )
   /SetVariables(Particles.GlowColor=&MultiplyLinearColor(
     Linear Color=$Particles.Initial.GlowColor
     Scale Factor=&MakeVector4(
      X=1
      Y=1
      Z=1
      W=&FloatFromCurve(
       FloatCurve=DataInterface() {propertyValues {Curve {
          defaultValue=1
          keys [{interpMode=RCIM_Cubic, tangentMode=RCTM_User, time=0.2127, value=0}, {interpMode=RCIM_Cubic, tangentMode=RCTM_User, time=0.4934, value=1}]
         }}}
       Scale Curve=0.2
      )
     )
    )) {isSetParameters=true, module=SetVariables_#9}
   &ScaleColor(
    Scale RGB=&VectorFromFloat(Value=&FloatFromCurve(FloatCurve=DataInterface() {propertyValues {Curve {keys [<at 0 1>,<at 0.9067 1>,<at 1 1.3>]}}}))
    ScaleA=true
    ScaleRGB=true
   ) {module=ScaleColor001}
   &Drag(
    Drag=&FloatFromCurve(
     FloatCurve=DataInterface() {propertyValues {Curve {
        defaultValue=1
        keys [
         {arriveTangent=4.1785, interpMode=RCIM_Cubic, leaveTangent=4.1785, tangentMode=RCTM_User, time=0, value=0}
         {arriveTangent=1.1179, interpMode=RCIM_Cubic, leaveTangent=1.1179, tangentMode=RCTM_User, time=0.2042, value=0.6228}
         {arriveTangent=0.0224, interpMode=RCIM_Cubic, leaveTangent=0.0224, tangentMode=RCTM_User, time=1, value=1}
        ]
       }}}
     Scale Curve=5
    )
    Use Linear Drag=true
    Use Rotational Drag=true
   )
   &PointAttractionForce(
    Attraction Radius=5000
    Attraction Strength=&FloatFromCurve(
     FloatCurve=DataInterface() {propertyValues {Curve {
        defaultValue=1
        keys [{interpMode=RCIM_Cubic, tangentMode=RCTM_User, time=0, value=0}, {interpMode=RCIM_Cubic, tangentMode=RCTM_User, time=0.7452, value=1}]
       }}}
     Scale Curve=200
    )
    Attractor Position:NiagaraPosition=&AddVectorToPosition(Position=$Engine.Emitter.SimulationPosition, Vector=<0 -80 130>)
    Kill Within Radius=false
    Module.Use Falloff=true
    UseAttractorOffset=false
   )
   &OffsetPosition(Position Offset=&MakeVector(
     X=0
     Y=&FloatFromCurve(
      FloatCurve=DataInterface() {propertyValues {Curve {
         defaultValue=1
         keys [
          {arriveTangent=-0.2069, interpMode=RCIM_Cubic, leaveTangent=-0.2069, tangentMode=RCTM_User, time=0, value=0}
          {arriveTangent=1.659, interpMode=RCIM_Cubic, leaveTangent=1.659, tangentMode=RCTM_User, time=1, value=1}
         ]
        }}}
      Scale Curve=-200
     )
     Z=&FloatFromCurve(
      FloatCurve=DataInterface() {propertyValues {Curve {
         defaultValue=1
         keys [
          {arriveTangent=4.5614, interpMode=RCIM_Cubic, leaveTangent=4.5614, tangentMode=RCTM_User, time=0.8, value=0}
          {arriveTangent=1.659, interpMode=RCIM_Cubic, leaveTangent=1.659, tangentMode=RCTM_User, time=0.9, value=1}
         ]
        }}}
      Scale Curve=30
     )
    )) {enabled=false}
   &OffsetPosition(Position Offset=&MakeVector(
     X=0
     Y=&FloatFromCurve(
      FloatCurve=DataInterface() {propertyValues {Curve {
         defaultValue=1
         keys [
          {arriveTangent=-0.2069, interpMode=RCIM_Cubic, leaveTangent=-0.2069, tangentMode=RCTM_User, time=0, value=0}
          {arriveTangent=1.659, interpMode=RCIM_Cubic, leaveTangent=1.659, tangentMode=RCTM_User, time=1, value=1}
         ]
        }}}
      Scale Curve=-50
     )
     Z=&FloatFromCurve(
      FloatCurve=DataInterface() {propertyValues {Curve {
         defaultValue=1
         keys [
          {arriveTangent=4.5614, interpMode=RCIM_Cubic, leaveTangent=4.5614, tangentMode=RCTM_User, time=0.8, value=0}
          {arriveTangent=1.659, interpMode=RCIM_Cubic, leaveTangent=1.659, tangentMode=RCTM_User, time=0.9, value=1}
         ]
        }}}
      Scale Curve=40
     )
    )) {module=OffsetPosition001}
   &SolveForcesAndVelocity(Clamp Velocity=false, Limit Acceleration=false)
  ]
  rendererClasses [&NiagaraMeshRendererProperties, &NiagaraSpriteRendererProperties]
  renderers [
   {rendererClass=&NiagaraMeshRendererProperties, rendererIndex=0}
   {rendererClass=&NiagaraSpriteRendererProperties, rendererIndex=1}
  ]
 }
 Skull_Form {
  bEnabled=false
  emitterUpdateScript [
   &EmitterState(Inactive Response=::NewEnumerator1, Life Cycle Mode=::NewEnumerator1, Loop Behavior=::NewEnumerator1, Loop Duration=1.35, UseLoopDelay=false)
   &SpawnBurst_Instantaneous(Spawn Count=23, Use Spawn Probability=false, UseLoopCountLimit=false)
  ]
  particleSpawnScript [
   &PartitionParticles()
   /SetVariables(Particles.VisibilityTag=$Emitter.SpawnGroup) {isSetParameters=true, module=SetVariables_#10}
   &InitializeParticle(
    Color Maximum=<rgba 1.25 1.1493 0.9825 1>
    Color Minimum=<rgba 0.6 0.6 0.6 1>
    Color Mode=::NewEnumerator2
    Lifetime=&MergeHelper_FloatSwitch(
     0=1.34
     1=&RandomRangeFloat(Minimum=0.7, Maximum=1)
     2=?
     3=?
     Index=$Particles.VisibilityTag
     Num=::NewEnumerator0
    )
    Mesh Scale Mode=::NewEnumerator3
    Mesh Uniform Scale=&MergeHelper_FloatSwitch(
     0=5
     1=&RandomRangeFloat(Minimum=1, Maximum=3)
     2=?
     3=?
     Index=$Particles.VisibilityTag
     Num=::NewEnumerator0
    )
    Sprite Size Mode=::NewEnumerator3
    Uniform Sprite Size=45
    UsePositionOffset=false
    Write Material Random=true
   )
   &SubUVAnimation(
    Mesh Renderer=DataInterface() {propertyValues {MeshRenderer=&self:Skull_Form_0.NiagaraMeshRendererProperties_0}}
    Output Alternate Renderer Binding=true
    Random Start Frame=true
    Renderer Type:ENiagaraMeshOrSprite=::NewEnumerator1
    SubUV Lookup Index Scale=0
    UseEndFrame=false
    UseStartFrame=false
   )
   &ShapeLocation(
    Box Size=<100 0 0>
    MaskSpawnGroup=false
    Offset=<0 110 0>
    Offset Mode=::NewEnumerator0
    Override Seed=false
    Shape Primitive=::NewEnumerator2
    Surface Only=false
    UseOwnerScale=true
   )
   &InitialMeshOrientation(
    Mesh Orientation Mode=::NewEnumerator2
    Rotation=&RandomRangeVector(Minimum=<-0.02 -0.02 -0.02>, Maximum=<0.02 0.02 0.02>)
    Use Rotation Vector=true
   )
   /SetVariables(Particles.GlowColor=<rgba 1 0.526 0.1188 1>) {isSetParameters=true, module=SetVariables_#8}
  ]
  particleUpdateScript [
   &ParticleState()
   &UpdateMeshOrientation()
   &SkeletalMeshLocation(
    Bone / Socket Index=3
    Bone Sampling Mode=::NewEnumerator4
    MaskSpawnGroup=false
    Position Sampling=::NewEnumerator2
    Sampled Position Offset=&Add_Vector(
     A=&Add_Vector(
      A=$User.SkullOffset
      B=&Multiply_Vector(
       A=&VectorFromFloat(Value=&FloatFromCurve(
         FloatCurve=DataInterface() {propertyValues {Curve {keys [
             <at 0 0.6>
             {arriveTangent=0.5785, interpMode=RCIM_Cubic, leaveTangent=0.5785, time=1.0397, value=0.6}
             {arriveTangent=-0.3749, interpMode=RCIM_Cubic, leaveTangent=-0.3749, time=1.21, value=1.3}
             {arriveTangent=-5.3168, interpMode=RCIM_Cubic, leaveTangent=-5.3168, time=1.3064, value=0.5}
             {interpMode=RCIM_Cubic, leaveTangent=0.7072, time=1.3417, value=0.6}
            ]}}}
         CurveIndex=$Emitter.Age
        ))
       B=&SelectVectorFromArray(
        Direct Array Index=&ReturnExecIndex()
        Vector Selection Array=DataInterface() {propertyValues {FloatData [
           <34 0.01 -57>
           <47 0.1 -35>
           <24 -0.02 -34.5>
           <0 0.12 -33>
           <64 -0.06 -27>
           <63 0.02 -5>
           <11.5 -0.12 -16>
           <12.5 0.13 8>
           <59.5 -0.03 18>
           <35 0.15 24.5>
           <15.5 0.08 29.5>
           <-1.5 -0.08 -58>
           <-33 -0.09 -57>
           <-25 -0.01 -34>
           <-48 -0.07 -35>
           <-63.5 0.05 -31>
           <-3 0.04 -16>
           <-63 0 -11.5>
           <-6.5 -0.14 9>
           <-9 0.04 29.5>
           <-31 -0.1 24>
           <-52 0.11 12.5>
           <-62 -0.05 6>
          ]}}
       )
      )
     )
     B=&Lerp_Vector(
      A=&MakeVector(
       X=0
       Y=0
       Z=&FloatFromCurve(
        FloatCurve=DataInterface() {propertyValues {Curve {
           defaultValue=1
           keys [
            {interpMode=RCIM_Cubic, tangentMode=RCTM_User, time=0.2538, value=0}
            {interpMode=RCIM_Cubic, tangentMode=RCTM_User, time=0.7073, value=1}
            {tangentMode=RCTM_User, time=1, value=0}
           ]
          }}}
        Scale Curve=7
       )
      )
      B=<0 0 0>
      Alpha=0
     )
    )
    Skeletal Mesh=$System.Skeletal Mesh Reader
   )
   &SkeletalMeshLocation(
    Bone / Socket Index=3
    Bone Sampling Mode=::NewEnumerator4
    MaskSpawnGroup=false
    Position Sampling=::NewEnumerator2
    Sampled Position Offset=&Add_Vector(
     A=&Add_Vector(
      A=<0 110 140>
      B=&Multiply_Vector(
       A=&VectorFromFloat(Value=0.6)
       B=&SelectVectorFromArray(
        Direct Array Index=&ReturnExecIndex()
        Vector Selection Array=DataInterface() {propertyValues {FloatData [
           <34 0.01 -57>
           <-1.5 -0.08 -58>
           <-34 -0.09 -57>
           <47 0.1 -35>
           <24 -0.02 -34.5>
           <0 0.12 -33>
           <-25 -0.01 -34>
           <-48 -0.07 -35>
           <64 -0.06 -27>
           <-63.5 0.05 -31>
           <63 0.02 -5>
           <11.5 -0.12 -16>
           <-3 -0.04 -16>
           <-63 0 -11.5>
           <12.5 0.13 8>
           <-6.5 -0.14 9>
           <59.5 -0.03 18>
           <35 0.15 24.5>
           <15.5 0.08 29.5>
           <-9 0.04 29.5>
           <-31 -0.1 24>
           <-52 0.11 12.5>
           <-62 -0.05 6>
           <0 0.09 0>
           <0 0.06 0>
           <0 0.07 0>
           <0 -0.11 0>
          ]}}
       )
      )
     )
     B=&Lerp_Vector(
      A=&RandomRangeVector(Minimum=<0 0 0>, Maximum=<0 0 0>)
      B=<0 0 0>
      Alpha=&FloatFromCurve(FloatCurve=DataInterface() {propertyValues {Curve {
          defaultValue=1
          keys [{arriveTangent=3.5381, interpMode=RCIM_Cubic, leaveTangent=3.5381, tangentMode=RCTM_User, time=0, value=0}, <at 0.3 1 RCIM_Cubic>]
         }}})
     )
    )
    Skeletal Mesh=$System.Skeletal Mesh Reader
   ) {enabled=false, module=SkeletalMeshLocation001}
   &ScaleMeshSize(Scale Factor=&VectorFromFloat(Value=&FloatFromCurve(
      FloatCurve=DataInterface() {propertyValues {Curve {
         defaultValue=1
         keys [
          {arriveTangent=4.1785, interpMode=RCIM_Cubic, leaveTangent=4.1785, tangentMode=RCTM_User, time=0, value=0}
          {arriveTangent=3.7835, interpMode=RCIM_Cubic, leaveTangent=3.7835, tangentMode=RCTM_User, time=0.2136, value=0.6422}
          {arriveTangent=0.0224, interpMode=RCIM_Cubic, leaveTangent=0.0224, tangentMode=RCTM_User, time=0.411, value=1}
         ]
        }}}
      Scale Curve=0.6
     )))
   &DynamicMaterialParameters(
    Index 0 Param 1=&FloatFromCurve(FloatCurve=DataInterface() {propertyValues {Curve {keys [<at 0 0>,<at 1 4>]}}})
    Index 0 Param 2=1
    Index 0 Param 3=&Waveform("Waveform [1]"=::NewEnumerator0, "[1] Frequency"=0.4, "[1] Amplitude Min/Max"=<1 2>)
    Module.Index0Float0Write=true
    Module.Index0Float1Write=true
    Module.Index0Float2Write=true
    Module.Index0Float3Write=true
   )
   &ScaleColor(
    Scale RGB=&Multiply_VectorByFloat(
     Vector=<1 1 1>
     Float=&Lerp_Float(
      A=&FloatFromCurve(FloatCurve=DataInterface() {propertyValues {Curve {keys [
           <at 0 1>
           {interpMode=RCIM_Cubic, tangentMode=RCTM_User, time=0.2854, value=1}
           {arriveTangent=-1.8481, tangentMode=RCTM_User, time=1, value=0.25}
          ]}}})
      B=&FloatFromCurve(FloatCurve=DataInterface() {propertyValues {Curve {keys [<at 0 2>,<at 0.0912 2>,<at 0.2854 2>,<at 0.9021 1.1441>,<at 1 1>]}}})
      Alpha=&MakeFloatFromInt(Int=&SpawnGroup())
     )
    )
    ScaleA=true
    ScaleRGB=true
   ) {enabled=false}
   /Game/Effects/Niagara/Modules/Curves/FN_HermiteSpline(
    Position Blend:NiagaraFloat=&RandomRangeFloat(
     Minimum=&FloatFromCurve(FloatCurve=DataInterface() {propertyValues {Curve {
         defaultValue=1
         keys [
          {arriveTangent=0.5789, interpMode=RCIM_Cubic, leaveTangent=0.5789, tangentMode=RCTM_User, time=0, value=0}
          {arriveTangent=5.7442, interpMode=RCIM_Cubic, leaveTangent=5.7442, tangentMode=RCTM_User, time=0.4819, value=1}
         ]
        }}})
     Maximum=&FloatFromCurve(FloatCurve=DataInterface() {propertyValues {Curve {
         defaultValue=1
         keys [{arriveTangent=6.4638, interpMode=RCIM_Cubic, leaveTangent=6.4638, tangentMode=RCTM_User, time=0, value=0}, <at 0.4765 1 RCIM_Cubic>]
        }}})
    )
    Position End:NiagaraPosition=$Particles.SkeletalMeshLocation.SampledPosition
    Position Start:NiagaraPosition=$Particles.Initial.ShapeLocation.ShapePosition
    Write Particles Position=true
   )
   &JitterPosition(Jitter Amount:NiagaraFloat=&FloatFromCurve(FloatCurve=DataInterface() {propertyValues {Curve {keys [<at 0.7565 0>,<at 0.8359 1>]}}}))
   /SetVariables(Particles.GlowColor=&MultiplyLinearColor(
     Linear Color=$Particles.Initial.GlowColor
     Scale Factor=&MakeVector4(
      X=1
      Y=1
      Z=1
      W=&FloatFromCurve(
       FloatCurve=DataInterface() {propertyValues {Curve {
          defaultValue=1
          keys [{interpMode=RCIM_Cubic, tangentMode=RCTM_User, time=0.2127, value=0}, {interpMode=RCIM_Cubic, tangentMode=RCTM_User, time=0.4934, value=1}]
         }}}
       Scale Curve=0.2
      )
     )
    )) {isSetParameters=true, module=SetVariables_#9}
   &self:WriteDeath(
    ID:NiagaraInt32=&ReturnExecIndex()
    Position Array:NiagaraDataInterfaceArrayPosition=$System.Position Array Init
   ) {module=WriteDeath}
   &ScaleColor(
    Scale RGB=&VectorFromFloat(Value=&FloatFromCurve(FloatCurve=DataInterface() {propertyValues {Curve {keys [<at 0 1>,<at 0.9067 1>,<at 1 1.3>]}}}))
    ScaleA=true
    ScaleRGB=true
   ) {module=ScaleColor001}
  ]
  rendererClasses [&NiagaraMeshRendererProperties, &NiagaraSpriteRendererProperties]
  renderers [
   {rendererClass=&NiagaraMeshRendererProperties, rendererIndex=0}
   {rendererClass=&NiagaraSpriteRendererProperties, rendererIndex=1}
  ]
 }
 Skull_Form_New {
  emitterUpdateScript [
   &EmitterState(
    Inactive Response=::NewEnumerator1
    Life Cycle Mode=::NewEnumerator1
    Loop Behavior=::NewEnumerator1
    Loop Duration Mode=::NewEnumerator1
    UseLoopDelay=false
   )
   &SpawnBurst_Instantaneous(Spawn Count=110, Use Spawn Probability=false, UseLoopCountLimit=false)
  ]
  particleSpawnScript [
   &PartitionParticles()
   &self:ScratchModule_03(
    Skull Offset:Vector3f=$User.SkullOffset
    Skull Scale:NiagaraFloat=$User.Skull Scale
    Static Mesh Reader=DataInterface() {propertyValues {
      DefaultMesh=&SM_Emote_VoyagerCuboid_SkullVerts4
      InstanceIndex=-1
      LODIndex=0
      PreviewMesh=&SM_Emote_VoyagerCuboid_SkullVerts4
      SoftSourceActor=null
      SourceMode=DefaultMeshOnly
      bAllowSamplingFromStreamingLODs=false
      bCaptureTransformsPerFrame=true
      bUsePhysicsBodyVelocity=false
      FilteredSockets []
      LODIndexUserParameter {parameter {name=None, typeDefHandle {registeredTypeIndex=108}, varData []}}
      MeshParameterBinding {parameter {name=None, typeDefHandle {registeredTypeIndex=1}, varData []}}
      SectionFilter {allowedMaterialSlots []}
     }}
    TriangleID:NiagaraInt32=&ReturnExecIndex()
   ) {module=ScratchModule_03}
   /SetVariables(Particles.VisibilityTag=$Emitter.SpawnGroup) {isSetParameters=true, module=SetVariables_#10}
   /SetVariables(Particles.Scale Array=DataInterface() {propertyValues {
      GpuSyncMode=SyncCpuToGpu
      MaxElements=0
      FloatData [2.6,2.1,2,2.8,2.6,1.8,1.9,1.2,2,1.1,1.1,3.2,1.3,1.7,1.5,2.5,2.5,2.4,1,0.9,1.7,0.9,0.9,3.5,1.5,1.7,1.5,1,1,1,1.5,1,1.5,1,2,3.5,3.5,1.2,1.4,2.5,2,1.3,1.2,1,2,2,2,3,2.9,2,1.9,2.7,1.2,2.6,2.55,3.2,1.5,1.4,3.5,3,3.2,3.5,3.7,2.7,1.6,1.5,2.3,2.2,1.2,1.2,1.6,2,2.1,2,2,1.6,3,2.5,3.7,3.5,1.3,2.6,3.3,3,3.5,4,3.2,4.5,4,2,2.7,2.5,2,2,1.5,2,2.6,1.8,1.7,1,2.5,2.8,1,2,1,1,1,1.3,1.25,2]
     }}) {isSetParameters=true, module=SetVariables_#7}
   &InitializeParticle(
    Color=&RandomRangeLinearColor(Minimum=<rgba 1 0.1623 0 1>, Maximum=<rgba 1 0.4363 0 1>)
    Lifetime=&MergeHelper_FloatSwitch(
     0=1.34
     1=&RandomRangeFloat(Minimum=0.7, Maximum=1)
     2=?
     3=?
     Index=$Particles.VisibilityTag
     Num=::NewEnumerator0
    )
    Mesh Scale Mode=::NewEnumerator3
    Mesh Uniform Scale=&Multiply_Float(
     A=&SelectFloatFromArray(Direct Array Index=&ReturnExecIndex(), Float Selection Array=$Particles.Scale Array)
     B=$User.Skull Scale
    )
    Position Offset=&Multiply_VectorByFloat(Vector=Linked() {linkedVariable {name=Particles.MeshIDPos, type=NiagaraPosition}}, Float=$User.Skull Scale)
    Sprite Size Mode=::NewEnumerator3
    Uniform Sprite Size=45
    UsePositionOffset=true
    Write Material Random=true
   )
   &SubUVAnimation(
    Mesh Renderer=DataInterface() {propertyValues {MeshRenderer=&self:Skull_Form_New_0.NiagaraMeshRendererProperties_0}}
    Output Alternate Renderer Binding=true
    Random Start Frame=true
    Renderer Type:ENiagaraMeshOrSprite=::NewEnumerator1
    SubUV Lookup Index Scale=0
    UseEndFrame=false
    UseStartFrame=false
   )
   /SetVariables(Particles.GlowColor=<rgba 1 0.526 0.1188 1>) {isSetParameters=true, module=SetVariables_#8}
   /SetVariables(Particles.Delay:NiagaraFloat=&RandomRangeFloat(Minimum=0, Maximum=0.5)) {isSetParameters=true}
  ]
  particleUpdateScript [
   &ParticleState()
   &ScaleMeshSize(Scale Factor=&VectorFromFloat(Value=&FloatFromCurve(
      FloatCurve=DataInterface() {propertyValues {Curve {
         defaultValue=1
         keys [
          {arriveTangent=4.1785, interpMode=RCIM_Cubic, leaveTangent=4.1785, tangentMode=RCTM_User, time=0, value=0}
          {arriveTangent=3.7835, interpMode=RCIM_Cubic, leaveTangent=3.7835, tangentMode=RCTM_User, time=0.2136, value=0.6422}
          {arriveTangent=0.0224, interpMode=RCIM_Cubic, leaveTangent=0.0224, tangentMode=RCTM_User, time=0.411, value=1}
         ]
        }}}
      CurveIndex=&Subtract_Float(
       A=$Particles.NormalizedAge
       B=&FloatFromCurve(FloatCurve=DataInterface() {propertyValues {Curve {keys [<at 0 1>,<at 1 0>]}}}, Scale Curve=$Particles.Delay)
      )
     )))
   &ScaleMeshSize(Scale Factor=&MakeVector(
     X=1
     Y=1
     Z=&Lerp_Float(
      A=1
      B=&FloatFromCurve(
       FloatCurve=DataInterface() {propertyValues {Curve {
          defaultValue=1
          keys [
           {interpMode=RCIM_Cubic, tangentMode=RCTM_User, time=0, value=1}
           {interpMode=RCIM_Cubic, tangentMode=RCTM_User, time=0.0875, value=10}
           {tangentMode=RCTM_User, time=0.4, value=1}
          ]
         }}}
       Scale Curve=&Lerp_Float(
        A=1
        B=&RandomRangeFloat(Minimum=1, Maximum=2)
        Alpha=&FloatFromCurve(FloatCurve=DataInterface() {propertyValues {Curve {
            defaultValue=1
            keys [
             {interpMode=RCIM_Cubic, tangentMode=RCTM_User, time=0, value=0}
             {interpMode=RCIM_Cubic, tangentMode=RCTM_User, time=0.1019, value=1}
             {tangentMode=RCTM_User, time=0.3, value=0}
            ]
           }}})
       )
      )
      Alpha=&FloatFromCurve(
       FloatCurve=DataInterface() {propertyValues {Curve {defaultValue=1, keys [{tangentMode=RCTM_User, time=0, value=0}, {tangentMode=RCTM_User, time=0.3, value=1}]}}}
       CurveIndex=&Subtract_Float(
        A=$Particles.NormalizedAge
        B=&FloatFromCurve(FloatCurve=DataInterface() {propertyValues {Curve {keys [<at 0 1>,<at 1 0>]}}}, Scale Curve=$Particles.Delay)
       )
      )
     )
    )) {module=ScaleMeshSize001}
   &DynamicMaterialParameters(
    Index 0 Param 1=1
    Index 0 Param 2=1
    Index 0 Param 3=&Waveform("Waveform [1]"=::NewEnumerator0, "[1] Frequency"=0.4, "[1] Amplitude Min/Max"=<1 2>)
    Module.Index0Float0Write=true
    Module.Index0Float1Write=true
    Module.Index0Float2Write=true
    Module.Index0Float3Write=true
   )
   /SetVariables(Particles.Position=&Multiply_Vector(
     A=Linked() {linkedVariable {name=Particles.Initial.Position, type=NiagaraPosition}}
     B=&MakeVector(
      X=1
      Y=1
      Z=&FloatFromCurve(
       FloatCurve=DataInterface() {propertyValues {Curve {
          defaultValue=1
          keys [<at 0 0 RCIM_Cubic>, {arriveTangent=2.9381, interpMode=RCIM_Cubic, leaveTangent=2.9381, time=0.2705, value=1}, <at 0.3404 1 RCIM_Cubic>]
         }}}
       CurveIndex=&Subtract_Float(
        A=$Particles.NormalizedAge
        B=&FloatFromCurve(FloatCurve=DataInterface() {propertyValues {Curve {keys [<at 0 1>,<at 1 0>]}}}, Scale Curve=$Particles.Delay)
       )
      )
     )
    )) {isSetParameters=true}
   /SetVariables(Particles.GlowColor=&MultiplyLinearColor(
     Linear Color=$Particles.Initial.GlowColor
     Scale Factor=&MakeVector4(
      X=1
      Y=1
      Z=1
      W=&FloatFromCurve(
       FloatCurve=DataInterface() {propertyValues {Curve {
          defaultValue=1
          keys [{interpMode=RCIM_Cubic, tangentMode=RCTM_User, time=0.2127, value=0}, {interpMode=RCIM_Cubic, tangentMode=RCTM_User, time=0.4934, value=1}]
         }}}
       Scale Curve=0.2
      )
     )
    )) {isSetParameters=true, module=SetVariables_#9}
   &ScaleColor(
    Scale RGB=&VectorFromFloat(Value=&FloatFromCurve(FloatCurve=DataInterface() {propertyValues {Curve {keys [<at 0 1>,<at 0.9067 1>,<at 1 1.3>]}}}))
    ScaleA=true
    ScaleRGB=true
   ) {module=ScaleColor001}
  ]
  rendererClasses [&NiagaraMeshRendererProperties, &NiagaraSpriteRendererProperties]
  renderers [
   {rendererClass=&NiagaraMeshRendererProperties, rendererIndex=0}
   {rendererClass=&NiagaraSpriteRendererProperties, rendererIndex=1}
   {rendererClass=&NiagaraMeshRendererProperties, rendererIndex=2}
  ]
 }
}
summary {
 systemName=NS_Emote_VoyagerCuboid_Skull
 emitters [
  {
   bEnabled=false
   emitterName=Skull_Form
   rendererClasses [&NiagaraMeshRendererProperties, &NiagaraSpriteRendererProperties]
  }
  {emitterName=Glitch, rendererClasses [&NiagaraSpriteRendererProperties]}
  {
   bEnabled=false
   emitterName=Skull_Break
   rendererClasses [&NiagaraMeshRendererProperties, &NiagaraSpriteRendererProperties]
  }
  {bEnabled=false, emitterName=Lightning_Break, rendererClasses [&NiagaraMeshRendererProperties]}
  {bEnabled=false, emitterName=Lightning_Form, rendererClasses [&NiagaraMeshRendererProperties]}
  {emitterName=Skull_Form_New, rendererClasses [&NiagaraMeshRendererProperties, &NiagaraSpriteRendererProperties]}
  {
   emitterName=Skull_Break_New
   rendererClasses [&NiagaraMeshRendererProperties, &NiagaraSpriteRendererProperties]
  }
  {emitterName=Dissolve_Glitch, rendererClasses [&NiagaraSpriteRendererProperties]}
 ]
 userVariables [
  User.bIsFrontEnd {type=NiagaraBool}
  User.bIsFrontEndPreview {type=NiagaraBool}
  User.BreakDelay {type=NiagaraFloat, defaultValue=1.3, description=""}
  User.BurstOffset {type=Vector3f, description="", defaultValue {struct=Vector3f, value=<0 90 150>}}
  User.OwnerLocalVelocity {type=Vector3f, description="", defaultValue {struct=Vector3f, value=<1 1 1>}}
  User.Skull Scale {type=NiagaraFloat, defaultValue=0.58, description=""}
  User.SkullOffset {type=Vector3f, description="", defaultValue {struct=Vector3f, value=<0 150 250>}}
  User.Variant {type=NiagaraFloat, defaultValue=0, description=""}
 ]
}
