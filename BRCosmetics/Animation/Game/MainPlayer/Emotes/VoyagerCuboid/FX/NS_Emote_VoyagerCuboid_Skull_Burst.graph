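# Niagara system for the VoyagerCuboid emote: a short, orange-gold skull burst effect timed by User.BreakDelay.
# Sprite glows and flares, spawn-group glow spheres, an expanding shockwave with a light, a mesh wave and radial lightning meshes form the burst.
# Most emitters follow User.BurstOffset, while Wave hardcodes its offset; some glow alphas scale by time of day.
# Skull_Break, which launches fragments from skeletal-mesh bone offsets to either side, is disabled.
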NiagaraSystem /BRCosmetics/Animation/Game/MainPlayer/Emotes/VoyagerCuboid/FX/NS_Emote_VoyagerCuboid_Skull_Burst
dependencies {
 usedDataInterfaces [
  &NiagaraDataInterfaceCurve
  &NiagaraDataInterfaceColorCurve
  &NiagaraDataInterfaceVectorCurve
  &NiagaraDataInterfaceMeshRendererInfo
  &NiagaraDataInterfaceSpriteRendererInfo
  &NiagaraDataInterfaceArrayFloat3
  &NiagaraDataInterfaceArrayPosition
  &NiagaraDataInterfaceSkeletalMesh
 ]
 usedDynamicInputs []
 usedModules [
  &self:SystemSpawnScript.SystemScriptSource.SystemScriptGraph.NiagaraNodeAssignment_3.SetVariables_#1
  &SystemState
  &self:SystemSpawnScript.SystemScriptSource.SystemScriptGraph.NiagaraNodeAssignment_0.SetVariables_#2
  &self:SystemSpawnScript.SystemScriptSource.SystemScriptGraph.NiagaraNodeAssignment_1.SetVariables_#3
  &self:SystemSpawnScript.SystemScriptSource.SystemScriptGraph.NiagaraNodeAssignment_2.SetVariables_#4
 ]
}
emitters {
 Glow001 {
  emitterUpdateScript [
   &EmitterState()
   &SpawnBurst_Instantaneous(Spawn Count=1, Spawn Time=$User.BreakDelay, Use Spawn Probability=false)
  ]
  particleSpawnScript [
   &InitializeParticle(
    Color=<rgba 0.6302 0.1197 0 0.6522>
    Lifetime=0.5
    Sprite Size Mode=::NewEnumerator3
    Uniform Sprite Size=600
    UsePositionOffset=false
    Write Material Random=true
   )
   &SystemLocation(MaskSpawnGroup=false, Offset=$User.BurstOffset)
  ]
  particleUpdateScript [
   &ParticleState()
   &ScaleSpriteSize(Scale Factor=&Vector2DFromFloat(Value=&FloatFromCurve(FloatCurve=DataInterface() {propertyValues {Curve {
         defaultValue=0
         keys [
          {interpMode=RCIM_Cubic, tangentMode=RCTM_User, time=0, value=0}
          <at 0.0627 0.9523 RCIM_Cubic>
          {arriveTangent=-2.5833, interpMode=RCIM_Cubic, leaveTangent=-2.5833, tangentMode=RCTM_User, time=1, value=0}
         ]
        }}})))
   &ScaleColor(
    Scale Alpha=&FloatFromCurve(
     FloatCurve=DataInterface() {propertyValues {Curve {
        defaultValue=0
        keys [{interpMode=RCIM_Cubic, tangentMode=RCTM_User, time=0, value=1}, {interpMode=RCIM_Cubic, tangentMode=RCTM_User, time=1, value=0}]
       }}}
     Scale Curve=&Multiply_Float(A=&FNTimeOfDayScalar(Night Scale=1, Morning Scale=0.5, Evening Scale=0.5), B=1)
    )
    ScaleA=true
    ScaleRGB=true
   )
   /SetVariables(
    Particles.GlowColor2:LinearColor=&MakeLinearColorFromVectorAndFloat(
     "Vector (RGB)"=&MakeVectorFromLinearColorRGB(Color=<rgba 5 3.1674 0.5186 1>)
     "Float (Alpha)"=&Multiply_Float(
      A=0.2
      B=&FloatFromCurve(
       FloatCurve=DataInterface() {propertyValues {Curve {
          defaultValue=0
          keys [{interpMode=RCIM_Cubic, tangentMode=RCTM_User, time=0, value=1}, {interpMode=RCIM_Cubic, tangentMode=RCTM_User, time=1, value=0}]
         }}}
       Scale Curve=&FNTimeOfDayScalar(Night Scale=0.3, Morning Scale=0.2, Evening Scale=0.5)
      )
     )
    )
    Particles.GlowSize2:Vector2f=&Vector2DFromFloat(Value=&Multiply_Float(
      A=700
      B=&FloatFromCurve(
       FloatCurve=DataInterface() {propertyValues {Curve {keys [
           {arriveTangent=0.4346, interpMode=RCIM_Cubic, leaveTangent=0.4346, tangentMode=RCTM_User, time=0, value=0}
           {arriveTangent=0.0473, interpMode=RCIM_Cubic, leaveTangent=0.0473, tangentMode=RCTM_User, time=0.2, value=1}
           {arriveTangent=-0.0078, interpMode=RCIM_Cubic, leaveTangent=-0.0078, tangentMode=RCTM_User, time=0.85, value=1}
           <at 1 0 RCIM_Cubic>
          ]}}}
       Scale Curve=&RandomRangeFloat(Minimum=0.7, Maximum=1, Evaluation Type=::NewEnumerator1)
      )
     ))
   ) {isSetParameters=true}
   /SetVariables(
    Particles.FlareColor=&ColorFromCurve(ColorCurve=DataInterface() {propertyValues {
       AlphaCurve {keys [<at 0.05 1 RCIM_Cubic>,<at 0.0751 0 RCIM_Cubic>]}
       BlueCurve {keys [<at 0 0 RCIM_Cubic>,<at 0.4851 0 RCIM_Cubic>]}
       GreenCurve {keys [<at 0 0.696 RCIM_Cubic>,<at 0.4851 0.2856 RCIM_Cubic>]}
       RedCurve {keys [<at 0 2 RCIM_Cubic>,<at 0.4851 1 RCIM_Cubic>]}
      }})
    Particles.Flaresize:Vector2f=&Multiply_Vector2D_ByFloat(
     Float=&Waveform("Waveform [1]"=::NewEnumerator12, "[1] Frequency"=10, "[1] Amplitude Min/Max"=<0.8 1>)
     Vector2D=&Multiply_Vector2D_ByFloat(
      Float=&FloatFromCurve(
       FloatCurve=DataInterface() {propertyValues {Curve {
          defaultValue=0
          keys [
           {
            arriveTangent=-17.7925
            arriveTangentWeight=1.5734
            interpMode=RCIM_Cubic
            leaveTangent=-17.7925
            leaveTangentWeight=1.5734
            tangentMode=RCTM_User
            tangentWeightMode=RCTWM_WeightedBoth
            time=0
            value=0.9985
           }
           {arriveTangent=-0.0356, interpMode=RCIM_Cubic, leaveTangent=-0.0356, tangentMode=RCTM_User, time=1, value=0}
          ]
         }}}
       Scale Curve=800
      )
      Vector2D=<0.65 1>
     )
    )
   ) {isSetParameters=true}
  ]
  rendererClasses [&NiagaraSpriteRendererProperties]
  renderers [
   {rendererClass=&NiagaraSpriteRendererProperties}
   {rendererClass=&NiagaraSpriteRendererProperties, rendererIndex=1}
  ]
 }
 GlowSphereMesh {
  emitterUpdateScript [
   /SetVariables(Emitter.Saturation=-1) {isSetParameters=true}
   &EmitterState()
   &SpawnBurst_Instantaneous(Spawn Count=1, Spawn Time=$User.BreakDelay, Use Spawn Probability=false)
   &SpawnBurst_Instantaneous(Spawn Count=1, Spawn Group=1, Spawn Time=$User.BreakDelay, Use Spawn Probability=false) {module=SpawnBurst_Instantaneous003}
   &SpawnBurst_Instantaneous(Spawn Count=1, Spawn Group=2, Spawn Time=$User.BreakDelay, Use Spawn Probability=false) {module=SpawnBurst_Instantaneous001}
  ]
  particleSpawnScript [
   &InitializeParticle(
    Color=<rgba 1 0.1565 0 1>
    Lifetime=0.5
    Mesh Scale Mode=::NewEnumerator3
    Mesh Uniform Scale=1.1
    Position Offset=$User.BurstOffset
    Position Offset Coordinate Space=::Local
    UsePositionOffset=true
    Write Material Random=true
   ) {module=InitializeParticle001}
   /SetVariables(Particles.Flare3Color:LinearColor=<rgba 3 0.8004 0.0526 1>, Particles.Flare3Size:Vector2f=<500 500>) {isSetParameters=true}
   /SetVariables(Particles.VisibilityTag=$Emitter.SpawnGroup) {isSetParameters=true}
  ]
  particleUpdateScript [
   &ParticleState()
   &DynamicMaterialParameters(
    Index 0 Param 1=&FloatFromCurve(FloatCurve=DataInterface() {propertyValues {Curve {keys [<at 0 0>,<at 0.22 0.5>,<at 1 0>]}}})
    Index 0 Param 2=&FloatFromCurve(FloatCurve=DataInterface() {propertyValues {Curve {keys [<at 0 15>,<at 1 0>]}}})
    Module.Param0WriteEnabled=true
    Module.Param1WriteEnabled=true
    Module.Param2WriteEnabled=true
    Module.Param3WriteEnabled=true
   )
   &ScaleMeshSize(Scale Factor=/Niagara/DynamicInputs/ScaleByCurve/ScaleVectorByCurve(
     InitialValue:Vector3f=<1 1 1>
     ScaleCurve:NiagaraDataInterfaceVectorCurve=DataInterface() {propertyValues {
       XCurve {keys [
         {arriveTangent=4.2859, interpMode=RCIM_Cubic, leaveTangent=4.2859, tangentMode=RCTM_User, time=0, value=0}
         {arriveTangent=2, interpMode=RCIM_Cubic, leaveTangent=2, tangentMode=RCTM_User, time=0.25, value=1}
         {interpMode=RCIM_Cubic, tangentMode=RCTM_User, time=1, value=2}
        ]}
       YCurve {keys [
         {arriveTangent=5.8758, interpMode=RCIM_Cubic, leaveTangent=5.8759, tangentMode=RCTM_User, time=0, value=0}
         {arriveTangent=2, interpMode=RCIM_Cubic, leaveTangent=2, tangentMode=RCTM_User, time=0.25, value=1}
         {interpMode=RCIM_Cubic, tangentMode=RCTM_User, time=1, value=2}
        ]}
       ZCurve {keys [
         {arriveTangent=4.7007, interpMode=RCIM_Cubic, leaveTangent=4.7007, tangentMode=RCTM_User, time=0, value=0}
         {arriveTangent=2, interpMode=RCIM_Cubic, leaveTangent=2, tangentMode=RCTM_User, time=0.25, value=1}
         {interpMode=RCIM_Cubic, tangentMode=RCTM_User, time=1, value=2}
        ]}
      }}
     CurveIndex:NiagaraFloat=$Particles.NormalizedAge
    ))
   &ScaleColor(
    Scale Alpha=&FloatFromCurve(
     FloatCurve=DataInterface() {propertyValues {Curve {keys [
         {arriveTangent=9.0725, interpMode=RCIM_Cubic, leaveTangent=9.0725, tangentMode=RCTM_User, time=0, value=0}
         {arriveTangent=-2.5249, interpMode=RCIM_Cubic, leaveTangent=-2.5249, tangentMode=RCTM_User, time=0.1408, value=0.9847}
         {arriveTangent=-0.0405, interpMode=RCIM_Cubic, leaveTangent=-0.0405, tangentMode=RCTM_User, time=1, value=0}
        ]}}}
     Scale Curve=1.25
    )
    Scale RGB=&VectorFromCurve(VectorCurve=DataInterface() {propertyValues {XCurve {keys [<at -0.0011 2>,<at 0.9988 1>]}, YCurve {keys [<at -0.0011 2>,<at 0.9988 1>]}, ZCurve {keys [<at -0.0011 2>,<at 0.9988 1>]}}})
    ScaleA=true
    ScaleRGB=true
   )
   &SpriteRotationRate(
    Delta Time:NiagaraFloat=&FloatFromCurve(FloatCurve=DataInterface() {propertyValues {Curve {keys [
         {tangentMode=RCTM_User, time=0, value=1}
         {arriveTangent=-0.3174, interpMode=RCIM_Cubic, leaveTangent=-0.3174, tangentMode=RCTM_User, time=0.3, value=1}
         {arriveTangent=0.0123, interpMode=RCIM_Cubic, leaveTangent=0.0123, tangentMode=RCTM_User, time=1, value=0}
        ]}}})
    Rotation Rate=&MaskFloatBySpawnGroup(Masked Float=5, Passthrough Float=0, Spawn Group Mask=2)
   )
   /SetVariables(
    Particles.Flare2Color:LinearColor=&MultiplyLinearColor(
     Linear Color=&ColorFromCurve(ColorCurve=DataInterface() {propertyValues {
        AlphaCurve {keys [<at 0 1 RCIM_Cubic>,<at 1 0 RCIM_Cubic>]}
        BlueCurve {keys [<at 0 0 RCIM_Cubic>,<at 0.5196 0.5306>,<at 1 1 RCIM_Cubic>]}
        GreenCurve {keys [<at 0 12.5217 RCIM_Cubic>,<at 0.5196 1.5187>,<at 1 1 RCIM_Cubic>]}
        RedCurve {keys [<at 0 40 RCIM_Cubic>,<at 0.5196 10>,<at 1 1 RCIM_Cubic>]}
       }})
     Scale Factor=&MakeFloat4FromVectorAndFloat(
      "Vector (RGB)"=&VectorFromFloat(Value=&FloatFromCurve(FloatCurve=DataInterface() {propertyValues {Curve {keys [<at 0 10 RCIM_Cubic>,<at 0.05 1>]}}}))
      "Float (Alpha)"=&FloatFromCurve(
       FloatCurve=DataInterface() {propertyValues {Curve {keys [
           {arriveTangent=14.7047, interpMode=RCIM_Cubic, leaveTangent=14.7047, tangentMode=RCTM_User, time=0.0011, value=0.0003}
           {arriveTangent=-1.6654, interpMode=RCIM_Cubic, leaveTangent=-1.6654, tangentMode=RCTM_User, time=0.05, value=1}
           {arriveTangent=0.3797, interpMode=RCIM_Cubic, leaveTangent=0.3797, tangentMode=RCTM_User, time=1, value=0}
          ]}}}
       Scale Curve=&FNTimeOfDayScalar(Night Scale=0.06, Morning Scale=0.7, Day Scale=0.8, Evening Scale=0.7)
      )
     )
    )
    Particles.Flare2Size:Vector2f=&Vector2DFromFloat(Value=&FloatFromCurve(
      FloatCurve=DataInterface() {propertyValues {Curve {keys [
          {arriveTangent=5.0321, interpMode=RCIM_Cubic, leaveTangent=5.0321, tangentMode=RCTM_User, time=0, value=0}
          {interpMode=RCIM_Cubic, tangentMode=RCTM_User, time=1, value=1}
         ]}}}
      Scale Curve=&Multiply_Float(A=400, B=0.6)
     ))
   ) {isSetParameters=true}
   /SetVariables(
    Particles.Flare3Color:LinearColor=&MultiplyLinearColor(
     Linear Color=$Particles.Initial.Flare3Color
     Scale Factor=&MakeVector4FromVectorAndFloat(
      "Vector (RGB)"=&MakeVectorFromLinearColorRGB(Color=&ColorFromCurve(ColorCurve=DataInterface() {propertyValues {
          AlphaCurve {keys []}
          BlueCurve {keys [{arriveTangent=-1.5872, interpMode=RCIM_Cubic, leaveTangent=-1.5872, tangentMode=RCTM_User, time=0, value=3}, <at 1 0.5 RCIM_Cubic>]}
          GreenCurve {keys [{arriveTangent=-4.9882, interpMode=RCIM_Cubic, leaveTangent=-4.9882, tangentMode=RCTM_User, time=0, value=3.75}, <at 1 0.625 RCIM_Cubic>]}
          RedCurve {keys [{arriveTangent=-3.5785, interpMode=RCIM_Cubic, leaveTangent=-3.5785, tangentMode=RCTM_User, time=0, value=4}, <at 1 1 RCIM_Cubic>]}
         }}))
      "Float (Alpha)"=&MaskFloatBySpawnGroup(
       Masked Float=2
       Passthrough Float=&MaskFloatBySpawnGroup(
        Masked Float=&FloatFromCurve(FloatCurve=DataInterface() {propertyValues {Curve {keys [<at 0 1>,<at 1 0>]}}}, Scale Curve=0.5)
        Passthrough Float=&FloatFromCurve(
         FloatCurve=DataInterface() {propertyValues {Curve {keys [
             {arriveTangent=-23.168, interpMode=RCIM_Cubic, leaveTangent=-23.168, tangentMode=RCTM_User, time=0, value=4}
             {arriveTangent=-4.5789, interpMode=RCIM_Cubic, leaveTangent=-4.5789, tangentMode=RCTM_User, time=0.2, value=1}
             <at 1 0 RCIM_Cubic>
            ]}}}
         Scale Curve=0.2
        )
        Spawn Group Mask=1
       )
      )
     )
    )
    Particles.Flare3Size:Vector2f=&Multiply_Vector2D(
     A=$Particles.Initial.Flare3Size
     B=&Vector2DFromFloat(Value=&FloatFromCurve(
       FloatCurve=DataInterface() {propertyValues {Curve {
          defaultValue=1
          keys [
           {arriveTangent=4.1785, interpMode=RCIM_Cubic, leaveTangent=4.1785, tangentMode=RCTM_User, time=0, value=0}
           {arriveTangent=1.1179, interpMode=RCIM_Cubic, leaveTangent=1.1179, tangentMode=RCTM_User, time=0.2042, value=0.6228}
           {arriveTangent=0.0224, interpMode=RCIM_Cubic, leaveTangent=0.0224, tangentMode=RCTM_User, time=1, value=1}
          ]
         }}}
       Scale Curve=1.3
      ))
    )
   ) {isSetParameters=true}
   /SetVariables(Particles.Lifetime=&Divide_Float(
     A=$Particles.Initial.Lifetime
     B=&MergeHelper_FloatSwitch(0=1.3, 1=1, 2=2, 3=?, Index=$Particles.Initial.VisibilityTag, Num=::NewEnumerator1)
    )) {isSetParameters=true}
  ]
  rendererClasses [&NiagaraMeshRendererProperties, &NiagaraSpriteRendererProperties]
  renderers [
   {rendererClass=&NiagaraMeshRendererProperties, rendererIndex=0}
   {rendererClass=&NiagaraSpriteRendererProperties, rendererIndex=1}
   {rendererClass=&NiagaraSpriteRendererProperties, rendererIndex=2}
  ]
 }
 Lightninglong {
  emitterUpdateScript [
   &EmitterState(Life Cycle Mode=::NewEnumerator1, Loop Behavior=::NewEnumerator1, Loop Delay=$User.BreakDelay, Loop Duration=0.3, UseLoopDelay=true)
   &SpawnRate(SpawnRate=20, Use Spawn Probability=false)
   &SpawnBurst_Instantaneous(Spawn Count=10, Use Spawn Probability=false, UseLoopCountLimit=false)
  ]
  particleSpawnScript [
   &PartitionParticles()
   &InitializeParticle(
    Color=<rgba 2 0.4991 0 1>
    Lifetime Max=0.4
    Lifetime Min=0.15
    Lifetime Mode=::NewEnumerator1
    Mesh Scale=&RandomRangeVector(Minimum=<1 0.5 0.5>, Maximum=<1.5 1.3 1.3>)
    Mesh Scale Mode=::NewEnumerator4
    UsePositionOffset=false
    Write Material Random=true
   )
   &CameraOffset(Camera Offset Amount=5)
   &ShapeLocation(MaskSpawnGroup=false, Offset=$User.BurstOffset, Offset Mode=::NewEnumerator0, Override Seed=false, Sphere Radius=1, UseOwnerScale=true)
   &InitialMeshOrientation(Mesh Orientation Mode=::NewEnumerator2, Orientation Vector=$Particles.ShapeLocation.ShapeNormal, Use Rotation Vector=true) {module=InitialMeshOrientation002}
   &InitialMeshOrientation(
    Mesh Orientation Mode=::NewEnumerator2
    Orientation Vector=$Particles.ShapeLocation.ShapeVector
    Rotation=&RandomRangeVector(Minimum=<0 0 0>, Maximum=<0 0 0>)
    Use Rotation Vector=true
   )
   /SetVariables(Particles.Random=&RandomRangeFloat(Minimum=0, Maximum=1)) {isSetParameters=true}
  ]
  particleUpdateScript [
   &ParticleState()
   &ScaleMeshSize(Scale Factor=&Multiply_VectorByFloat(Vector=<1 1 1>, Float=0.6))
   &ScaleColor(
    Scale Alpha=&Multiply_Float(
     A=10
     B=&FloatFromCurve(FloatCurve=DataInterface() {propertyValues {Curve {keys [<at 0 0>,<at 0.1 0>,<at 0.2 0.8>,<at 1 0>]}}})
    )
    Scale RGB=&Multiply_VectorByFloat(
     Vector=&Multiply_VectorByFloat(Vector=<1 1 1>, Float=1)
     Float=&Multiply_Float(
      A=&RandomRangeFloat(Minimum=0, Maximum=1, Evaluation Type=::NewEnumerator1)
      B=&Multiply_Float(
       A=1
       B=&FloatFromCurve(FloatCurve=DataInterface() {propertyValues {Curve {keys [<at 0 10>,<at 1 10>]}}})
      )
     )
    )
    ScaleA=true
    ScaleRGB=true
   )
   &DynamicMaterialParameters(
    Index 0 Param 1=&FloatFromCurve(FloatCurve=DataInterface() {propertyValues {Curve {keys [{tangentMode=RCTM_User, time=0, value=1}, {tangentMode=RCTM_User, time=1, value=1}]}}})
    Index 0 Param 2=0.7
    Index 0 Param 3=&FloatFromCurve(FloatCurve=DataInterface() {propertyValues {Curve {keys [
         {arriveTangent=4.0072, interpMode=RCIM_Cubic, leaveTangent=4.0072, tangentMode=RCTM_User, time=0, value=0}
         {arriveTangent=5, interpMode=RCIM_Cubic, leaveTangent=5, time=0.2543, value=0.6225}
         {arriveTangent=3.1884, interpMode=RCIM_Cubic, leaveTangent=3.1884, time=0.4, value=2}
         {arriveTangent=1.6993, interpMode=RCIM_Cubic, leaveTangent=1.6993, tangentMode=RCTM_User, time=1, value=3}
        ]}}})
    Index 0 Param 4=5
    Index 1 Param 1=&Lerp_Float(A=1, B=2, Alpha=$Particles.Random)
    Index 1 Param 2=$Particles.Random
    Index 1 Param 3=2
    Index 1 Param 4=1
    Index 2 Param 1=&FloatFromCurve(FloatCurve=DataInterface() {propertyValues {Curve {keys [<at 0 1>,<at 1 0.3>]}}})
    Module.Param0WriteEnabled=true
    Module.Param0WriteEnabled1=true
    Module.Param0WriteEnabled2=true
    Module.Param1WriteEnabled=true
    Module.Param1WriteEnabled1=true
    Module.Param1WriteEnabled2=true
    Module.Param2WriteEnabled=true
    Module.Param2WriteEnabled1=true
    Module.Param2WriteEnabled2=true
    Module.Param3WriteEnabled=true
    Module.Param3WriteEnabled1=true
    Module.Param3WriteEnabled2=true
    Write Parameter Index 1=true
    Write Parameter Index 2=true
   )
   &SolveForcesAndVelocity(Clamp Velocity=false, Limit Acceleration=false)
   /SetVariables(Particles.Lightning glow color:LinearColor=&MakeLinearColorFromVectorAndFloat(
     "Vector (RGB)"=&MakeVectorFromLinearColorRGB(Color=&MultiplyLinearColor(
       Linear Color=<rgba 2 0.427 0.1586 1>
       Scale Factor=&MakeVector4_FromFloat(Float=&RandomRangeFloat(Minimum=0.5, Maximum=1, Evaluation Type=::NewEnumerator1))
      ))
     "Float (Alpha)"=&Multiply_Float(
      A=&FloatFromCurve(FloatCurve=DataInterface() {propertyValues {Curve {keys [<at 0 1 RCIM_Cubic>, {interpMode=RCIM_Cubic, leaveTangent=-3.6307, time=0.9, value=0}]}}})
      B=7
     )
    )) {isSetParameters=true}
   /SetVariables(Particles.Glow Dy02:Vector4f=&MakeVector4(
     X=&FloatFromCurve(FloatCurve=DataInterface() {propertyValues {Curve {keys [{tangentMode=RCTM_User, time=0, value=0.2}, {tangentMode=RCTM_User, time=0.5, value=1}]}}})
     Y=0.7
     Z=0
     W=5
    )) {isSetParameters=true}
   /SetVariables(Particles.Glow Dy03:Vector4f=&MakeVector4(
     X=&FloatFromCurve(
      FloatCurve=DataInterface() {propertyValues {Curve {keys [<at 0 1>,<at 1 0>]}}}
      Scale Curve=&Lerp_Float(A=2, B=2, Alpha=$Particles.Random)
     )
     Y=$Particles.Random
     Z=2
     W=1
    )) {isSetParameters=true}
  ]
  rendererClasses [&NiagaraMeshRendererProperties]
  renderers [
   {rendererClass=&NiagaraMeshRendererProperties, rendererIndex=0}
   {rendererClass=&NiagaraMeshRendererProperties, rendererIndex=1}
  ]
 }
 Shockwave {
  emitterUpdateScript [
   &EmitterState()
   &SpawnBurst_Instantaneous(Spawn Count=1, Spawn Time=$User.BreakDelay, Use Spawn Probability=false)
  ]
  particleSpawnScript [
   &InitializeParticle(
    Color=<rgba 5000 1104.5792 208.3871 3>
    Lifetime=0.3
    Mesh Scale Mode=::NewEnumerator3
    Mesh Uniform Scale=0.3
    Position Offset=$User.BurstOffset
    Position Offset Coordinate Space=::Local
    UsePositionOffset=true
    Write Material Random=true
   ) {module=InitializeParticle001}
   &CameraOffset()
  ]
  particleUpdateScript [
   &ParticleState()
   &DynamicMaterialParameters(
    Index 0 Param 1=&FloatFromCurve(FloatCurve=DataInterface() {propertyValues {Curve {keys [
         {arriveTangent=2.0031, interpMode=RCIM_Cubic, leaveTangent=2.0031, tangentMode=RCTM_User, time=-0.0072, value=-0.0078}
         {arriveTangent=1.9873, interpMode=RCIM_Cubic, leaveTangent=1.9873, tangentMode=RCTM_User, time=0.6336, value=1.3072}
         <at 1.0021 1.9784>
         <at 1.61 1.9784>
        ]}}})
    Index 0 Param 2=&FloatFromCurve(FloatCurve=DataInterface() {propertyValues {Curve {keys [<at -0.0005 15>,<at 0.1612 13.3874>,<at 1.0021 5>,<at 1.61 5>]}}})
    Index 0 Param 3=-1
    Module.Param0WriteEnabled=true
    Module.Param1WriteEnabled=true
    Module.Param2WriteEnabled=true
    Module.Param3WriteEnabled=true
   )
   &ScaleMeshSize(Scale Factor=&VectorFromCurve(VectorCurve=DataInterface() {propertyValues {
       XCurve {keys [
         {arriveTangent=20.6325, interpMode=RCIM_Cubic, leaveTangent=20.6325, tangentMode=RCTM_User, time=0, value=0.7}
         {arriveTangent=4.6893, interpMode=RCIM_Cubic, leaveTangent=4.6893, tangentMode=RCTM_User, time=0.3596, value=6.0271}
         {arriveTangent=2.5984, tangentMode=RCTM_User, time=1.0043, value=8.3622}
         <at 1.61 9>
        ]}
       YCurve {keys [
         {arriveTangent=21.0918, interpMode=RCIM_Cubic, leaveTangent=21.0918, tangentMode=RCTM_User, time=0, value=0.7}
         {arriveTangent=4.6893, interpMode=RCIM_Cubic, leaveTangent=4.6893, tangentMode=RCTM_User, time=0.3582, value=6.0747}
         {arriveTangent=2.5984, tangentMode=RCTM_User, time=1.0043, value=8.3622}
         <at 1.61 9>
        ]}
       ZCurve {keys [
         {arriveTangent=21.4939, interpMode=RCIM_Cubic, leaveTangent=21.4939, tangentMode=RCTM_User, time=0, value=0.7}
         {arriveTangent=4.6893, interpMode=RCIM_Cubic, leaveTangent=4.6893, tangentMode=RCTM_User, time=0.3615, value=6.1467}
         {arriveTangent=2.5984, tangentMode=RCTM_User, time=1.0043, value=8.3622}
         <at 1.61 9>
        ]}
      }}))
   &ScaleColor(
    Scale Alpha=&FloatFromCurve(FloatCurve=DataInterface() {propertyValues {Curve {keys [
         {interpMode=RCIM_Cubic, tangentMode=RCTM_User, time=0, value=-0.061}
         {arriveTangent=3.5976, interpMode=RCIM_Cubic, leaveTangent=3.5976, time=0.022, value=-0.0318}
         {arriveTangent=0.2603, interpMode=RCIM_Cubic, leaveTangent=0.2603, time=0.05, value=0.1188}
         {arriveTangent=-0.1606, interpMode=RCIM_Cubic, leaveTangent=-0.1606, tangentMode=RCTM_User, time=0.3345, value=0.0495}
         {interpMode=RCIM_Cubic, tangentMode=RCTM_User, time=0.9972, value=-0.0001}
        ]}}})
    Scale RGB=&ScaleFloatByCurve(
     InitialValue:Vector3f=<1 2 1>
     ScaleCurve:NiagaraDataInterfaceVectorCurve=DataInterface() {propertyValues {
       XCurve {keys [<at 0 0>,<at 0.0388 1.9605>,<at 1 1>]}
       YCurve {keys [<at 0 0>,<at 0.0358 1.422>,<at 1.0017 0.264>]}
       ZCurve {keys [<at 0 0>,<at 0.0417 0.9748>,<at 1.0017 0.264>]}
      }}
     CurveIndex:NiagaraFloat=$Particles.NormalizedAge
    )
    ScaleA=true
    ScaleRGB=true
   )
   /SetVariables(
    Particles.LightBColor:LinearColor=&ColorFromCurve(ColorCurve=DataInterface() {propertyValues {
       AlphaCurve {keys [<at 0 1 RCIM_Cubic>,<at 1 0 RCIM_Cubic>]}
       BlueCurve {keys [<at 0 1.9713 RCIM_Cubic>,<at 0.5196 0.7549>,<at 1 1 RCIM_Cubic>]}
       GreenCurve {keys [<at 0 13.8759 RCIM_Cubic>,<at 0.5196 3.649>,<at 1 1 RCIM_Cubic>]}
       RedCurve {keys [<at 0 40 RCIM_Cubic>,<at 0.5196 10>,<at 1 1 RCIM_Cubic>]}
      }})
    Particles.LightBRadius:NiagaraFloat=&FloatFromCurve(
     FloatCurve=DataInterface() {propertyValues {Curve {keys [
         <at -0.0284 132.9428>
         {arriveTangent=65.2061, interpMode=RCIM_Cubic, leaveTangent=65.2061, tangentMode=RCTM_User, time=0.2862, value=174.5049}
         <at 1 200>
        ]}}}
     Scale Curve=0.6
    )
   ) {isSetParameters=true}
   /SetVariables(
    Particles.FlowColor:LinearColor=&MultiplyLinearColor(
     Linear Color=<rgba 50 14.9513 5.7077 1>
     Scale Factor=&MakeFloat4FromVectorAndFloat(
      "Vector (RGB)"=<1 1 1>
      "Float (Alpha)"=&FloatFromCurve(FloatCurve=DataInterface() {propertyValues {Curve {keys [{arriveTangent=-2.1056, interpMode=RCIM_Cubic, leaveTangent=-2.1056, tangentMode=RCTM_User, time=0, value=1}, <at 0.9 0>]}}})
     )
    )
    Particles.FlowDynamic:Vector4f=&MakeVector4(X=0, Y=1, Z=1, W=0)
    Particles.FlowSize:Vector3f=&VectorFromFloat(Value=&Multiply_Float(
      A=3
      B=&FloatFromCurve(FloatCurve=DataInterface() {propertyValues {Curve {keys [
           {arriveTangent=1.9852, interpMode=RCIM_Cubic, leaveTangent=1.9852, tangentMode=RCTM_User, time=0, value=0.15}
           {arriveTangent=0.0045, interpMode=RCIM_Cubic, leaveTangent=0.0045, tangentMode=RCTM_User, time=1, value=1.05}
          ]}}})
     ))
   ) {enabled=false, isSetParameters=true}
  ]
  rendererClasses [&NiagaraMeshRendererProperties, &NiagaraLightRendererProperties]
  renderers [
   {rendererClass=&NiagaraMeshRendererProperties, rendererIndex=0}
   {rendererClass=&NiagaraLightRendererProperties, rendererIndex=1}
  ]
 }
 Skull_Break {
  bEnabled=false
  emitterUpdateScript [
   &EmitterState()
   &SpawnBurst_Instantaneous(Spawn Count=50, Spawn Time=$User.BreakDelay, Use Spawn Probability=false, UseLoopCountLimit=false)
  ]
  particleSpawnScript [
   &PartitionParticles()
   &InitializeParticle(
    Color Maximum=<rgba 1.25 1.1493 0.9825 1>
    Color Minimum=<rgba 0.6 0.6 0.6 1>
    Color Mode=::NewEnumerator2
    Lifetime=&RandomRangeFloat(Minimum=0.2, Maximum=0.6)
    Mesh Scale Mode=::NewEnumerator3
    Mesh Uniform Scale=&RandomRangeFloat(Minimum=0.8, Maximum=1.6)
    Sprite Size Mode=::NewEnumerator3
    Uniform Sprite Size=35
    UsePositionOffset=false
    Write Material Random=true
   )
   &SubUVAnimation(
    Mesh Renderer=DataInterface() {propertyValues {MeshRenderer=&self:Skull_Break_0.NiagaraMeshRendererProperties_0}}
    Output Alternate Renderer Binding=true
    Random Start Frame=true
    Renderer Type:ENiagaraMeshOrSprite=::NewEnumerator1
    SubUV Lookup Index Scale=0
    UseEndFrame=false
    UseStartFrame=false
   )
   /SetVariables(Particles.SideToLaunch:NiagaraInt32=&RandomRangeInt(Minimum=0, Maximum=1)) {isSetParameters=true}
   /SetVariables(Particles.Group:NiagaraInt32=&MergeHelper_IntSwitch(
     0=&RandomRangeInt(Minimum=0, Maximum=10)
     1=&RandomRangeInt(Minimum=12, Maximum=23)
     2=?
     3=?
     4=?
     5=?
     Index=$Particles.SideToLaunch
     Num=::NewEnumerator0
    )) {isSetParameters=true}
   &SkeletalMeshLocation(
    Bone / Socket Index=3
    Bone Sampling Mode=::NewEnumerator4
    MaskSpawnGroup=false
    Sampled Position Offset=&Add_Vector(
     A=&Add_Vector(
      A=$User.SkullOffset
      B=&Multiply_Vector(
       A=&VectorFromFloat(Value=0.6)
       B=&SelectVectorFromArray(
        Direct Array Index=$Particles.SideToLaunch
        Vector Selection Array=DataInterface() {propertyValues {FloatData [
           <34 0.01 -57>
           <47 0.1 -35>
           <24 -0.02 -34.5>
           <0 0.12 -33>
           <64 -0.06 -27>
           <63 0.02 -5>
           <11.5 -0.12 -16>
           <12.5 0.13 8>
           <59.5 -0.03 18>
           <35 0.15 24.5>
           <15.5 0.08 29.5>
           <-1.5 -0.08 -58>
           <-33 -0.09 -57>
           <-25 -0.01 -34>
           <-48 -0.07 -35>
           <-63.5 0.05 -31>
           <-3 0.04 -16>
           <-63 0 -11.5>
           <-6.5 -0.14 9>
           <-9 0.04 29.5>
           <-31 -0.1 24>
           <-52 0.11 12.5>
           <-62 -0.05 6>
          ]}}
       )
      )
     )
     B=&Lerp_Vector(
      A=&RandomRangeVector(Minimum=<0 0 0>, Maximum=<0 0 0>)
      B=<0 0 0>
      Alpha=&FloatFromCurve(FloatCurve=DataInterface() {propertyValues {Curve {
          defaultValue=1
          keys [{arriveTangent=3.5381, interpMode=RCIM_Cubic, leaveTangent=3.5381, tangentMode=RCTM_User, time=0, value=0}, <at 0.3 1 RCIM_Cubic>]
         }}})
     )
    )
    Skeletal Mesh=$System.Skeletal Mesh Reader
   )
   &AddVelocity(Velocity=&MakeVector(
     X=&Multiply_Float(
      A=&RandomRangeFloat(Minimum=-80, Maximum=-150)
      B=&MergeHelper_FloatSwitch(0=-4, 1=4, 2=?, 3=?, Index=$Particles.SideToLaunch, Num=::NewEnumerator0)
     )
     Y=&RandomRangeFloat(Minimum=-80, Maximum=-500)
     Z=&RandomRangeFloat(Minimum=10, Maximum=250)
    ))
   &MeshRotationForce(Rotation=&RandomRangeVector(Minimum=<15 15 15>, Maximum=<-15 -15 -15>))
   &ApplyInitialForces()
   /SetVariables(Particles.GlowColor=<rgba 1 0.526 0.1188 1>) {isSetParameters=true}
  ]
  particleUpdateScript [
   &ParticleState()
   &ScaleMeshSize(Scale Factor=&VectorFromFloat(Value=&FloatFromCurve(
      FloatCurve=DataInterface() {propertyValues {Curve {
         defaultValue=0
         keys [{interpMode=RCIM_Cubic, tangentMode=RCTM_User, time=0.686, value=1}, {interpMode=RCIM_Cubic, tangentMode=RCTM_User, time=1, value=0}]
        }}}
      Scale Curve=0.6
     )))
   &DynamicMaterialParameters(
    Index 0 Param 1=&FloatFromCurve(FloatCurve=DataInterface() {propertyValues {Curve {keys [<at 0 0>,<at 1 4>]}}})
    Index 0 Param 2=1
    Index 0 Param 3=&Waveform("Waveform [1]"=::NewEnumerator0, "[1] Frequency"=0.4, "[1] Amplitude Min/Max"=<1 2>)
    Index 3 Param 1=&MakeFloatFromInt(Int=$Particles.Initial.SideToLaunch)
    Module.Index0Float0Write=true
    Module.Index0Float1Write=true
    Module.Index0Float2Write=true
    Module.Index0Float3Write=true
    Module.Index3Float0Write=true
    Module.Index3Float1Write=false
    Module.Index3Float2Write=false
    Module.Index3Float3Write=false
    Write Parameter Index 3=::NewEnumerator1
   )
   &GravityForce(Gravity=&MakeVector(X=0, Y=0, Z=&RandomRangeFloat(Minimum=-550, Maximum=-980)))
   &Drag(
    Drag=&FloatFromCurve(
     FloatCurve=DataInterface() {propertyValues {Curve {keys [<at 0 0>,<at 0.3018 3>]}}}
     Scale Curve=&RandomRangeFloat(Minimum=0.5, Maximum=1.5)
    )
    Use Linear Drag=true
    Use Rotational Drag=true
   )
   &SolveForcesAndVelocity(Clamp Velocity=false, Limit Acceleration=false)
   /SetVariables(Particles.GlowColor=&MultiplyLinearColor(
     Linear Color=$Particles.Initial.GlowColor
     Scale Factor=&MakeVector4(
      X=1
      Y=1
      Z=1
      W=&FloatFromCurve(
       FloatCurve=DataInterface() {propertyValues {Curve {
          defaultValue=1
          keys [{interpMode=RCIM_Cubic, tangentMode=RCTM_User, time=0.2127, value=1}, {interpMode=RCIM_Cubic, tangentMode=RCTM_User, time=0.4934, value=0}]
         }}}
       Scale Curve=0.2
      )
     )
    )) {isSetParameters=true}
   &ScaleColor(
    Scale RGB=&VectorFromFloat(Value=&FloatFromCurve(
      FloatCurve=DataInterface() {propertyValues {Curve {
         defaultValue=1
         keys [{interpMode=RCIM_Cubic, tangentMode=RCTM_User, time=0, value=1}, {interpMode=RCIM_Cubic, tangentMode=RCTM_User, time=1, value=0}]
        }}}
      Scale Curve=&RandomRangeFloat(Minimum=1.5, Maximum=2.5)
     ))
    ScaleA=true
    ScaleRGB=true
   )
  ]
  rendererClasses [&NiagaraMeshRendererProperties]
  renderers [{rendererClass=&NiagaraMeshRendererProperties, rendererIndex=0}]
 }
 Wave {
  emitterSpawnScript [/SetVariables() {isSetParameters=true}]
  emitterUpdateScript [
   &EmitterState()
   &SpawnBurst_Instantaneous(Spawn Count=1, Spawn Time=$User.BreakDelay, Use Spawn Probability=false)
  ]
  particleSpawnScript [
   &InitializeParticle(
    Color=<rgba 4 2.0891 0.0047 1>
    Lifetime=0.3
    Mesh Scale Mode=::NewEnumerator3
    Mesh Uniform Scale=1.1
    Position Offset=<0 90 140>
    UsePositionOffset=true
    Write Material Random=true
   )
   &CameraOffset()
   &InitialMeshOrientation(Mesh Orientation Mode=::NewEnumerator2, Orientation Vector=<0 0 1>, Use Rotation Vector=false)
   /SetVariables(Particles.SphereScale=0.1) {isSetParameters=true}
  ]
  particleUpdateScript [
   &ParticleState()
   &DynamicMaterialParameters(
    Index 0 Param 1=&FloatFromCurve(FloatCurve=DataInterface() {propertyValues {Curve {keys [{arriveTangent=-2.1015, interpMode=RCIM_Cubic, leaveTangent=-2.1015, tangentMode=RCTM_User, time=0, value=0.5}, <at 1 -0.15>]}}})
    Index 0 Param 2=&FloatFromCurve(FloatCurve=DataInterface() {propertyValues {Curve {keys [{arriveTangent=-0.3464, interpMode=RCIM_Cubic, leaveTangent=-0.3464, tangentMode=RCTM_User, time=0, value=1}, <at 1 0.9>]}}})
    Index 0 Param 3=&FloatFromCurve(FloatCurve=DataInterface() {propertyValues {Curve {keys [<at 0 0>,<at 0.15 1>,<at 1 1.2>]}}})
    Module.Param0WriteEnabled=true
    Module.Param1WriteEnabled=true
    Module.Param2WriteEnabled=true
    Module.Param3WriteEnabled=false
   )
   &ScaleColor(
    Scale Alpha=&FloatFromCurve(FloatCurve=DataInterface() {propertyValues {Curve {keys [<at 0 1>,<at 1 0>]}}})
    Scale RGB=&VectorFromFloat(Value=&FloatFromCurve(FloatCurve=DataInterface() {propertyValues {Curve {keys [<at 0 10>,<at 0.15 3>,<at 1 1>]}}}))
    ScaleA=true
    ScaleRGB=true
   ) {module=ScaleColor001}
   &ScaleMeshSize(Scale Factor=&VectorFromFloat(Value=&FloatFromCurve(
      FloatCurve=DataInterface() {propertyValues {Curve {keys [<at 0 0>,<at 0.1 1.4>,<at 1 2>]}}}
      Scale Curve=&Multiply_Float(A=1, B=1)
     )))
   /SetVariables(
    Particles.SphereColor:LinearColor=&MultiplyLinearColor(
     Linear Color=<rgba 1 0.2215 0 1>
     Scale Factor=&MakeFloat4FromVectorAndFloat(
      "Vector (RGB)"=<1 1 1>
      "Float (Alpha)"=&FloatFromCurve(
       FloatCurve=DataInterface() {propertyValues {Curve {keys [
           <at 0 1>
           {interpMode=RCIM_Cubic, tangentMode=RCTM_User, time=0.05, value=1}
           {interpMode=RCIM_Cubic, tangentMode=RCTM_User, time=0.5, value=0}
          ]}}}
       Scale Curve=10
      )
     )
    )
    Particles.SphereDynamicParam:Vector4f=&MakeVector4(X=1, Y=0, Z=1, W=1)
    Particles.SphereScale:NiagaraFloat=&FloatFromCurve(
     FloatCurve=DataInterface() {propertyValues {Curve {keys [
         {arriveTangent=4.1785, interpMode=RCIM_Cubic, leaveTangent=4.1785, tangentMode=RCTM_User, time=0, value=0.05}
         {arriveTangent=1.7974, interpMode=RCIM_Cubic, leaveTangent=1.7974, tangentMode=RCTM_User, time=0.15, value=0.6}
         {arriveTangent=-0.0405, interpMode=RCIM_Cubic, leaveTangent=-0.0405, tangentMode=RCTM_User, time=0.5, value=1}
        ]}}}
     Scale Curve=0.5
    )
    Particles.SphereScalevector:Vector3f=&VectorFromFloat(Value=$Particles.SphereScale)
   ) {isSetParameters=true}
  ]
  rendererClasses [&NiagaraMeshRendererProperties]
  renderers [
   {rendererClass=&NiagaraMeshRendererProperties, rendererIndex=0}
   {rendererClass=&NiagaraMeshRendererProperties, rendererIndex=1}
  ]
 }
}
summary {
 systemName=NS_Emote_VoyagerCuboid_Skull_Burst
 emitters [
  {emitterName=Glow001, rendererClasses [&NiagaraSpriteRendererProperties]}
  {emitterName=Shockwave, rendererClasses [&NiagaraMeshRendererProperties, &NiagaraLightRendererProperties]}
  {emitterName=GlowSphereMesh, rendererClasses [&NiagaraMeshRendererProperties, &NiagaraSpriteRendererProperties]}
  {emitterName=Wave, rendererClasses [&NiagaraMeshRendererProperties]}
  {bEnabled=false, emitterName=Skull_Break, rendererClasses [&NiagaraMeshRendererProperties]}
  {emitterName=Lightninglong, rendererClasses [&NiagaraMeshRendererProperties]}
 ]
 userVariables [
  User.bIsFrontEnd {type=NiagaraBool}
  User.bIsFrontEndPreview {type=NiagaraBool}
  User.BreakDelay {type=NiagaraFloat, defaultValue=1.3, description=""}
  User.BurstOffset {type=Vector3f, description="", defaultValue {struct=Vector3f, value=<0 90 140>}}
  User.OwnerLocalVelocity {type=Vector3f, description="", defaultValue {struct=Vector3f, value=<1 1 1>}}
  User.SkullOffset {type=Vector3f, description="", defaultValue {struct=Vector3f, value=<0 110 160>}}
  User.Variant {type=NiagaraFloat, defaultValue=0, description=""}
 ]
}
